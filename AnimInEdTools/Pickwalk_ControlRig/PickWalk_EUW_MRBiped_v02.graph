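# Editor utility widget for pickwalking control selection on MR biped Control Rigs in Sequencer, with sidestep, mirror, part-select and recall actions on buttons and hotkeys.
# Walking reads selected controls from the focused Level Sequence's rigs, then splits them by numeric suffix.
# Numbered controls step along name-based chains, falling back with all others to SET_Map lookups.
# IKFK_Check reads switch values at the current frame to choose between mapped IK and FK controls.
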EditorUtilityWidgetBlueprint /AnimInEdTools/Pickwalk_ControlRig/PickWalk_EUW_MRBiped_v02
graphs [
 AddSetToSelection {nodes [
   n1 = AddSetToSelection!() -> (then, selectedController:Name, setToCheck:Array of Names) {is=title K2Node_FunctionEntry}
   > n2 = For Each Loop with Break!(Exec:Exec, Array:Array of Names=@n1.setToCheck, Break:Exec=@n6.then) -> (LoopBody:Exec, Array Element:Name, Array Index:int, Completed:Exec)
   n3 = Length(TargetArray:Array of Names=Get setToCheck() -> (setToCheck:Array of Names))
   > Branch(Condition="In Range (Integer)"!(Value:int=@n2.Array Index, Min:int=1, Max:int=int - int(A=@n3, B=2), InclusiveMin:bool=true, InclusiveMax:bool=true) -> (ReturnValue:bool))
   > Branch(Condition="Equal (Name)"(A=@n2, B=Get selectedController() -> (selectedController:Name=None)))
   > n4 = For Each Loop(Array:Array of Names=Get setToCheck()) -> (Array Element:Name)
   n5 = Length(TargetArray=Get setToCheck())
   > Branch(Condition="In Range (Integer)"(Value=@n4.Array Index, Min=1, Max=int - int(A=@n5, B=2), InclusiveMin=true, InclusiveMax=true))
   > n6 = Add(TargetArray:Array of Names=Get Selection_New() -> (Selection_New:Array of Names), NewItem:"Name (by ref)"=@n4) {is=title K2Node_CallArrayFunction}
  ]}
 AddToSelection_Smart {nodes [
   AddToSelection_Smart() {is=title K2Node_FunctionEntry}
   > n1 = For Each Loop(Array=Get Selection_New())
   > n2 = Set ParsedControlToSelect(ParsedControlToSelect:Name=@n1) -> (Output_Get:Name=None)
   > DEBUG!(execute, self, NewParam:String=Append(A="parsing: ", B="To String (Name)"(InName=@n2))) -> (then)
   > n3 = For Each Loop(Array:Array of Control Rig Object References=Get ControlRigs_Input() -> (ControlRigs_Input:Array of Control Rig Object References)) -> (Array Element:Control Rig Object Reference)
   > Set ParsedControlRig(ParsedControlRig:Control Rig Object Reference=@n3) -> (Output_Get:Control Rig Object Reference)
   > n4 = Branch(Condition=OR Boolean(A="Equal (Name)"(A=@n1, B=hand_r_ctrl), B="Equal (Name)"(A=@n1, B=arm_r_ctrl)))
   > n5 = Set IK_AttrToParse(IK_AttrToParse:Name=Arm R IK) -> (Output_Get:Name=None)
   n6 = Set IKFK_RunCheck(execute=[@n21.then, @n5.then, @n19.then, @n17.then], IKFK_RunCheck:bool=true) -> (Output_Get:bool)
   n7 = Branch(execute=[@n22.then, @n6.then], Condition=Get IKFK_RunCheck() -> (IKFK_RunCheck:bool))
   n8 = Get IK_AttrToParse() -> (IK_AttrToParse:Name=None)
   > DEBUG(NewParam=Append!(A="IK detected: ", B="To String (Name)"(InName=@n8), C:String="  checking if it's on or not") -> (ReturnValue))
   > n9 = GetLocalControlRigBool!(
    execute
    LevelSequence:Level Sequence Object Reference=Get CurrLevelSequence() -> (CurrLevelSequence:Level Sequence Object Reference)
    ControlRig:Control Rig Object Reference=Get ParsedControlRig() -> (ParsedControlRig:Control Rig Object Reference)
    ControlName:Name=@n8
    Frame:Frame Number Structure
    TimeUnit:EMovieSceneTimeUnit Enum=DisplayRate
   ) -> (then, ReturnValue:bool)
   > n10 = Sequence()
   > Branch(Condition=@n9)
   n11 = "To String (Name)"(InName=Get ParsedControlToSelect() -> (ParsedControlToSelect:Name=None))
   > n12 = Branch(Condition=Contains(SearchIn=@n11, Substring=leg))
   > Set ParsedControlToSelect(ParsedControlToSelect=String To Name(InString=Replace(SourceString=@n11, From=leg_, To=leg_ik_))) -> (Output_Get=None)
   n13 = "To String (Name)"(InName=Get ParsedControlToSelect() -> (ParsedControlToSelect=None))
   n14 = Branch(execute=@n12.else, Condition=Contains(SearchIn=@n13, Substring=hand))
   > Set ParsedControlToSelect(ParsedControlToSelect=String To Name(InString=Replace(SourceString=@n13, From=hand_, To=hand_ik_))) -> (Output_Get=None)
   n15 = "To String (Name)"(InName=Get ParsedControlToSelect() -> (ParsedControlToSelect=None))
   Branch(execute=@n14.else, Condition=Contains(SearchIn=@n15, Substring=arm))
   > Set ParsedControlToSelect(ParsedControlToSelect=String To Name(InString=Replace(SourceString=@n15, From=arm_, To=hand_ik_))) -> (Output_Get=None)
   SelectControl!(
    execute=[@n10.then_1, @n7.else]
    self:Control Rig Object Reference=Get ParsedControlRig()
    InControlName:"Name (by ref)"=Get ParsedControlToSelect() -> (ParsedControlToSelect=None)
    bSelect:bool=true
    bSetupUndo:bool=false
   ) -> (then)
   n16 = Branch(execute=@n4.else, Condition=OR Boolean(A="Equal (Name)"(A=@n1, B=Arm L/FK 2), B="Equal (Name)"(A=@n1, B=Arm L/IK)))
   > n17 = Set IK_AttrToParse(IK_AttrToParse=Arm L IK) -> (Output_Get=None)
   n18 = Branch(execute=@n16.else, Condition="Equal (Name)"(A=@n1, B=leg_r_ctrl))
   > n19 = Set IK_AttrToParse(IK_AttrToParse=Leg R IK) -> (Output_Get=None)
   n20 = Branch(execute=@n18.else, Condition="Equal (Name)"(A=@n1, B=leg_l_ctrl))
   > n21 = Set IK_AttrToParse(IK_AttrToParse=Leg L IK) -> (Output_Get=None)
   n22 = Set IKFK_RunCheck(execute=@n20.else, IKFK_RunCheck=false)
   n23 = Set Hierarchy(Hierarchy:Rig Hierarchy Object Reference=GetHierarchy!(self:Control Rig Object Reference=Get In ControlRig() -> (In ControlRig:Object Reference)) -> (ReturnValue:Rig Hierarchy Object Reference)) -> (Output_Get:Rig Hierarchy Object Reference)
   > n24 = For Each Loop!(
    Exec
    Array:Array of Rig Element Key Structures=GetChildren!(
     self:Rig Hierarchy Object Reference=Get Hierarchy() -> (Hierarchy:Rig Hierarchy Object Reference)
     InKey:Rig Element Key Structure=Find Control!(self:Rig Hierarchy Object Reference=@n23, InKey_Type:ERigElementType Enum=None, InKey_Name:Name=global_ctrl) -> (
      ReturnValue_PreferredEulerAngles:Rig Preferred Euler Angles Structure
      ReturnValue_TopologyControlSettings:Control Settings Topology Structure
      ReturnValue_Key:Rig Element Key Structure
      ReturnValue_Index:int
      ReturnValue_SubIndex:int
      ReturnValue_CreatedAtInstructionIndex:int
      ReturnValue_SpawnIndex:int
     ).ReturnValue_Key
     bRecursive:bool=false
    ) -> (ReturnValue:Array of Rig Element Key Structures)
   ) -> (LoopBody, Array Element_Type:ERigElementType Enum, Array Element_Name:Name, Array Index, Completed)
   > Branch(Condition=Contains Item(TargetArray=Get IK_Attrs() -> (IK_Attrs:Array of Names), ItemToFind=@n24.Array Element_Name))
   > Branch()
   > GetLocalControlRigBool(LevelSequence=Get CurrLevelSequence(), ControlRig=Get ParsedControlRig(), ControlName=@n24.Array Element_Name, TimeUnit=DisplayRate)
   "To String (Name)"(InName=@n24.Array Element_Name)
  ]}
 ControlExistCheck {nodes [
   n1 = ControlExistCheck!() -> (then, Control Rig:Control Rig Object Reference, ControlName_In:Name) {is=title K2Node_FunctionEntry}
   > n2 = Set Hierarchy(Hierarchy=GetHierarchy(self=@n1))
   n3 = Get Control Settings!(self:Rig Hierarchy Object Reference=@n2, InKey_Type:ERigElementType Enum=Control, InKey_Name:Name=@n1.ControlName_In) -> (
    ReturnValue_AnimationType:ERigControlAnimationType Enum
    ReturnValue_ControlType:ERigControlType Enum
    ReturnValue_DisplayName:Name
    ReturnValue_PrimaryAxis:ERigControlAxis Enum
    ReturnValue_LimitEnabled:Array of Rig Control Limit Enabled Structures
    ReturnValue_bDrawLimits:bool=~
    ReturnValue_MinimumValue:Rig Control Value Structure
    ReturnValue_MaximumValue:Rig Control Value Structure
    ReturnValue_bShapeVisible:bool=~
    ReturnValue_ShapeVisibility:ERigControlVisibility Enum
    ReturnValue_ShapeName:Name
    ReturnValue_ShapeColor:Linear Color Structure
    ReturnValue_bIsTransientControl:bool=~
    ReturnValue_ControlEnum:Enum Object Reference
    ReturnValue_Customization:Rig Control Element Customization Structure
    ReturnValue_DrivenControls:Array of Rig Element Key Structures
    ReturnValue_bGroupWithParentControl:bool=~
    ReturnValue_bRestrictSpaceSwitching:bool=~
    ReturnValue_FilteredChannels:Array of ERigControlTransformChannel Enums
    ReturnValue_PreferredRotationOrder:EEulerRotationOrder Enum
    ReturnValue_bUsePreferredRotationOrder:bool=~
    ReturnValue_ShapeTransform:Transform
   ) {is=title K2Node_CallFunction}
   > n4 = Branch(Condition="Equal Exactly (String)"(A=Animation Control, B=Enum to String!(Enumerator:ERigControlAnimationType Enum=@n3) -> (ReturnValue:String)))
   > DEBUG(NewParam=Append!(
     A="Settings Preferred Rotation Order is: "
     B=Enum to String(Enumerator:EEulerRotationOrder Enum=@n3.ReturnValue_PreferredRotationOrder)
     C=" ** use preferred rotation order == "
     D:String="To String (Boolean)"(InBool=@n3.ReturnValue_bUsePreferredRotationOrder)
    ) -> (ReturnValue))
   > Return Node!(execute, ControlExists:bool=true, ControlName_Out:Name=@n1.ControlName_In)
   Return Node(execute=@n4.else, ControlExists=false, ControlName_Out=None)
  ]}
 DEBUG {nodes [
   n1 = DEBUG!() -> (then, NewParam:String) {is=title K2Node_FunctionEntry}
   > n2 = Branch(Condition=Get Debug_Check() -> (Debug_Check:bool))
   > n3 = PrintString(InString=@n1)
   Return Node!(execute=[@n3.then, @n2.else])
  ]}
 EventGraph {nodes [
   GetSelection!(execute, self, ClearSelection?:bool=true) -> (then, GetSelection_Output:Array of STR Control Rig Name Pair Structures)
   > n1 = For Each Loop!(Exec, Array:Array of Names=Get Selection_Input() -> (Selection_Input:Array of Names)) -> (LoopBody, Array Element:Name, Array Index, Completed)
   n2 = For Each Loop(Exec=@n1.Completed, Array:Array of Control Rig Object References=Get ControlRigs_Input()) -> (Array Element:Control Rig Object Reference)
   > ClearControlSelection!(execute, self:Control Rig Object Reference=@n2, bSetupUndo:bool=false) -> (then, ReturnValue:bool)
   > n3 = For Each Loop(Array:Array of Names=Get Selection_New()) -> (Array Element:Name)
   > SelectControl(self=@n2, InControlName=@n3, bSelect=true, bSetupUndo=false)
   Set selectedControl_parsed(execute=@n1.LoopBody, selectedControl_parsed:Name=@n1) -> (Output_Get:Name=None)
   > n4 = Pickwalk_Parse!(
    execute
    self
    selectedControl:Name=Get selectedControl_parsed() -> (selectedControl_parsed:Name=None)
    bodySet:Array of Names=Get Arm_L_FK() -> (Arm_L_FK:Array of Names)
   ) -> (then, False:bool)
   > n5 = Branch(Condition=@n4)
   n6 = Pickwalk_Parse(execute=@n5.else, selectedControl=Get selectedControl_parsed() -> (selectedControl_parsed=None), bodySet=Get Core() -> (Core:Array of Names))
   > n7 = Branch(Condition=@n6)
   n8 = Pickwalk_Parse(
    execute=@n7.else
    selectedControl=Get selectedControl_parsed() -> (selectedControl_parsed=None)
    bodySet=Get Arm_R_FK() -> (Arm_R_FK:Array of Names)
   )
   > n9 = Branch(Condition=@n8)
   n10 = Pickwalk_Parse(
    execute=@n9.else
    selectedControl=Get selectedControl_parsed() -> (selectedControl_parsed=None)
    bodySet=Get Arm_L_Attach() -> (Arm_L_Attach:Array of Names)
   )
   > n11 = Branch(Condition=@n10)
   n12 = Pickwalk_Parse(
    execute=@n11.else
    selectedControl=Get selectedControl_parsed() -> (selectedControl_parsed=None)
    bodySet=Get Arm_R_Attach() -> (Arm_R_Attach:Array of Names)
   )
   > n13 = Branch(Condition=@n12)
   n14 = Pickwalk_Parse(
    execute=@n13.else
    selectedControl=Get selectedControl_parsed() -> (selectedControl_parsed=None)
    bodySet=Get Leg_L_FK() -> (Leg_L_FK:Array of Names)
   )
   > n15 = Branch(Condition=@n14)
   n16 = Pickwalk_Parse(
    execute=@n15.else
    selectedControl=Get selectedControl_parsed() -> (selectedControl_parsed=None)
    bodySet=Get Leg_R_FK() -> (Leg_R_FK:Array of Names)
   )
   > n17 = Branch(Condition=@n16)
   n18 = Pickwalk_Parse(
    execute=@n17.else
    selectedControl=Get selectedControl_parsed() -> (selectedControl_parsed=None)
    bodySet=Get Arm_L_IK() -> (Arm_L_IK:Array of Names)
   )
   > n19 = Branch(Condition=@n18)
   n20 = Pickwalk_Parse(
    execute=@n19.else
    selectedControl=Get selectedControl_parsed() -> (selectedControl_parsed=None)
    bodySet=Get Arm_R_IK() -> (Arm_R_IK:Array of Names)
   )
   > n21 = Branch(Condition=@n20)
   n22 = Pickwalk_Parse(
    execute=@n21.else
    selectedControl=Get selectedControl_parsed() -> (selectedControl_parsed=None)
    bodySet=Get Leg_L_IK() -> (Leg_L_IK:Array of Names)
   )
   > n23 = Branch(Condition=@n22)
   n24 = Pickwalk_Parse(
    execute=@n23.else
    selectedControl=Get selectedControl_parsed() -> (selectedControl_parsed=None)
    bodySet=Get Leg_R_IK() -> (Leg_R_IK:Array of Names)
   )
   > n25 = Branch(Condition=@n24)
   n26 = Pickwalk_Face!(execute=@n25.else, self) -> (then, faceCtrl:bool)
   > n27 = Branch(Condition=@n26)
   n28 = Get selectedControl_parsed() -> (selectedControl_parsed=None)
   Branch(execute=@n27.else, Condition=OR Boolean(A="Equal (Name)"(A=leg_l_ctrl, B=@n28), B="Equal (Name)"(A=@n28, B=leg_r_ctrl)))
   > n29 = Branch(Condition="Equal (Integer)"(A=Get Direction() -> (Direction:int=0), B=1))
   > Add(TargetArray=Get Selection_New(), NewItem=body_ctrl)
   Add(execute=@n29.else, TargetArray=Get Selection_New(), NewItem=offset_ctrl)
   "On Clicked (BTN_PW_Up)"() {is=title K2Node_ComponentBoundEvent}
   > PickWalk_MASTER!(execute, self, Direction:int=1) -> (then)
   "On Pressed (BTN_PW_Down)"() {is=title K2Node_ComponentBoundEvent}
   > PickWalk_MASTER(Direction=-1)
   "On Clicked (BTN_PW_Left)"() {is=title K2Node_ComponentBoundEvent}
   > SideStep!(execute, self, Direction:int=-1) -> (then)
   "On Clicked (BTN_PW_Right)"() {is=title K2Node_ComponentBoundEvent}
   > SideStep(Direction=1)
   "On Clicked (BTN_SelPart)"() {is=title K2Node_ComponentBoundEvent}
   > Select_Part()
   "On Clicked (BTN_Mirror)"() {is=title K2Node_ComponentBoundEvent}
   > Select_Mirror()
   "On Clicked (Test_Bl)"() {is=title K2Node_ComponentBoundEvent}
   > n30 = Branch(Condition="Equal (Boolean)"(A:bool=Get Debug_Check(), B:bool) -> (ReturnValue:bool=~) {is=title K2Node_PromotableOperator})
   > Set Debug_Check(Debug_Check:bool=true) -> (Output_Get:bool)
   > GetSelection(ClearSelection?=false)
   Set Debug_Check(execute=@n30.else, Debug_Check=false)
   "On Clicked (BTN_RecallSel)"() {is=title K2Node_ComponentBoundEvent}
   > Recall_LastInput()
   "On Clicked (BTN_SpecialShit)"() {is=title K2Node_ComponentBoundEvent}
   > SpecialShit()
   Event Construct()
   > SetIsEnabled(bInIsEnabled=true)
   > SetIsEnabled(bInIsEnabled=true)
   AddToSelection_Smart() {is=title K2Node_CallFunction}
   Event PreConstruct()
   Event Tick()
   Branch(Condition="integer > integer"(A=Length(TargetArray:Array of Wildcards)))
  ]}
 GetSelection {nodes [
   n1 = GetSelection!() -> (then, ClearSelection?:bool=~) {is=title K2Node_FunctionEntry}
   > RESET Vars()
   > n2 = GetFocusedLevelSequence!(execute) -> (then, ReturnValue:Level Sequence Object Reference)
   > n3 = Set CurrLevelSequence(CurrLevelSequence:Level Sequence Object Reference=@n2) -> (Output_Get:Level Sequence Object Reference)
   > n4 = GetControlRigs!(execute, LevelSequence:Level Sequence Object Reference=@n3) -> (then, ReturnValue:Array of Control Rig Sequencer Binding Proxy Structures)
   > n5 = For Each Loop(Array:Array of Control Rig Sequencer Binding Proxy Structures=@n4) -> (Array Element:Control Rig Sequencer Binding Proxy Structure)
   n6 = Sequence!(execute=@n5.Completed) -> (then_0, then_1, then_2:Exec)
   > DEBUG(NewParam=*****CONTROL SELECTION INPUT:*****) {is=title K2Node_CallFunction}
   > n7 = For Each Loop!(
    Exec
    Array:Array of STR Control Rig Name Pair Structures=Get STR_Selection_Input() -> (STR_Selection_Input:Array of STR Control Rig Name Pair Structures)
   ) -> (
    LoopBody
    Array Element_ControlRig_2_#1:Control Rig Object Reference
    Array Element_ControlName_5_#2:Name
    Array Index
    Completed
   )
   > DEBUG(NewParam="To String (Name)"(InName=@n7.Array Element_ControlName_5_#2))
   Branch(execute=@n6.then_1, Condition=@n1)
   > n8 = For Each Loop(Array=Get STR_Selection_Input())
   > ClearControlSelection(self=@n8, bSetupUndo=false)
   Return Node!(execute=@n6.then_2, GetSelection_Output:Array of STR Control Rig Name Pair Structures=Get STR_Selection_Input())
   n9 = Set Control Rig(
    execute=@n5.LoopBody
    Control Rig:Control Rig Object Reference=Break Control Rig Sequencer Binding Proxy!(ControlRigSequencerBindingProxy:"Control Rig Sequencer Binding Proxy Structure (by ref)"=@n5) -> (
     Proxy:Movie Scene Binding Proxy Structure
     ControlRig:Control Rig Object Reference
     Track:Movie Scene Control Rig Parameter Track Object Reference
    ) {is=title K2Node_BreakStruct}.ControlRig
   ) -> (Output_Get:Control Rig Object Reference)
   > n10 = For Each Loop!(Exec, Array:Array of Names=CurrentControlSelection!(self:Control Rig Object Reference=@n9) -> (ReturnValue:Array of Names)) -> (LoopBody, Array Element:Name, Array Index, Completed)
   > n11 = ControlExistCheck!(execute, self, Control Rig:Control Rig Object Reference=@n9, ControlName_In:Name=@n10) -> (then, ControlExists:bool, ControlName_Out:Name=None) {is=title K2Node_CallFunction}
   > Branch(Condition=@n11)
   > Add Unique!(
    execute
    TargetArray:Array of STR Control Rig Name Pair Structures=Get STR_Selection_Input()
    NewItem_ControlRig_2_#1:Control Rig Object Reference=@n9
    NewItem_ControlName_5_#2:Name=@n10
   ) -> (then, ReturnValue)
  ]}
 Get_L_or_R {nodes [
   n1 = Get_L_or_R!() -> (then, in_sourceControl:Name, in_array:Array of Names) {is=title K2Node_FunctionEntry}
   > n2 = For Each Loop(Array=@n1.in_array)
   n3 = Split(SourceString=Split(SourceString="To String (Name)"(InName=@n1), InStr=/, SearchDir=FromStart), InStr=" ", SearchDir=FromStart)
   Return Node!(execute=@n2.Completed, out_LorR:String=@n3.RightS, out_array:Array of Names=Get L_or_R_Array() -> (L_or_R_Array:Array of Names))
   n4 = "To String (Name)"(InName=@n2)
   > Add Unique!(
    execute
    TargetArray:Array of Names=Get L_or_R_Array()
    NewItem:"Name (by ref)"=String To Name(InString=Replace(SourceString=@n4, From=Q, To=@n3.RightS))
   ) -> (then, ReturnValue)
  ]}
 IKFK_Check {nodes [
   n1 = IKFK_Check!() -> (then, control to check:Name, controlRigToCheck:Control Rig Object Reference, SelectControl:bool=true) {is=title K2Node_FunctionEntry}
   > Set ControlToCheck_local(ControlToCheck_local:Name=@n1) -> (Output_Get:Name=None)
   > Set ControlRigToCheck_local(ControlRigToCheck_local:Control Rig Object Reference=@n1.controlRigToCheck) -> (Output_Get:Control Rig Object Reference)
   n2 = Get ControlToCheck_local() -> (ControlToCheck_local:Name=None)
   > DEBUG(NewParam=Append!(A="*****IK FK CHECK REQUEST FOR - ", B="To String (Name)"(InName=@n2), C=" - CHECKING IK*****") -> (ReturnValue))
   > n3 = Get Current Local Time!(execute, TimeUnit:EMovieSceneTimeUnit Enum=DisplayRate) -> (
    then
    ReturnValue_Frame_FrameNumber:Frame Number Structure
    ReturnValue_Frame_SubFrame:float=~
    ReturnValue_Time:float=~
    ReturnValue_MarkedFrame:String
    ReturnValue_Timecode:Timecode Structure
    ReturnValue_PositionType:EMovieScenePositionType Enum
    ReturnValue_UpdateMethod:EUpdatePositionMethod Enum
    ReturnValue_bHasJumped:bool=~
   ) {is=title K2Node_CallFunction}
   n4 = Split(SourceString="To String (Name)"(InName=Get ControlToCheck_local() -> (ControlToCheck_local=None)), InStr=/, SearchDir=FromEnd)
   n5 = Split(SourceString=@n4, InStr=" ", SearchDir=FromStart)
   > n6 = Branch(Condition="Equal Exactly (String)"(A=@n5, B=Clavicle))
   > n7 = Set IKFK_IKFKSwitchControl(IKFK_IKFKSwitchControl:Name=String To Name(InString=Append(A="Arm ", B=@n5.RightS, C=/Ik Fk Switch))) -> (Output_Get:Name=None)
   n8 = GetLocalControlRigBool(
    execute=[@n7.then, @n18.then, @n19.then]
    LevelSequence=Get CurrLevelSequence()
    ControlRig=Get ControlRigToCheck_local() -> (ControlRigToCheck_local:Control Rig Object Reference)
    ControlName=Get IKFK_IKFKSwitchControl() -> (IKFK_IKFKSwitchControl:Name=None)
    Frame=@n3
    TimeUnit=DisplayRate
   )
   > Set loc_ikFk(loc_ikFk:bool=@n8) -> (Output_Get:bool)
   n9 = FrameNumber to Integer!(InFrameNumber:"Frame Number Structure (by ref)"=@n3) -> (ReturnValue:int=0)
   > DEBUG(NewParam=Append!(A="Ik Switch is: ", B="To String (Boolean)"(InBool=@n8), C=" at frame; ", D:String="To String (Integer)"(InInt=@n9)) -> (ReturnValue))
   > n10 = Branch(Condition=@n1.SelectControl)
   > n11 = For Each Loop!(Exec, Array:Array of STR Pickwalk Map IKFK Structures=Get SET_Map_IKIFK() -> (SET_Map_IKIFK:Array of STR Pickwalk Map IKFK Structures)) -> (
    LoopBody
    Array Element_targetControl_2_#3:Name
    Array Element_destinationControl_4_#4:Name
    Array Index
    Completed
   )
   n12 = Get ControlToCheck_local() -> (ControlToCheck_local=None)
   > DEBUG(NewParam=Append!(
     A="do select ---control to check:  "
     B="To String (Name)"(InName=@n12)
     C="    target Control: "
     D="To String (Name)"(InName=@n11)
     E:String="    dest control:"
     F:String="To String (Name)"(InName=@n11.Array Element_destinationControl_4_#4)
    ) -> (ReturnValue))
   n13 = Get ControlToCheck_local() -> (ControlToCheck_local=None)
   > Branch(Condition=OR Boolean(A="Equal (Name)"(A=@n11, B=@n12), B="Equal (Name)"(A=@n11.Array Element_destinationControl_4_#4, B=@n13)))
   > n14 = Branch(Condition=Get loc_ikFk() -> (loc_ikFk:bool))
   > Add Unique!(
    execute
    TargetArray:Array of STR Control Rig Name Pair Structures=Get STR_Selection_Output() -> (STR_Selection_Output:Array of STR Control Rig Name Pair Structures)
    NewItem_ControlRig_2_#1:Control Rig Object Reference=Get ControlRigToCheck_local()
    NewItem_ControlName_5_#2:Name=@n11.Array Element_destinationControl_4_#4
   ) -> (then, ReturnValue)
   Add Unique(
    execute=@n14.else
    TargetArray=Get STR_Selection_Output()
    NewItem_ControlRig_2_#1=Get ControlRigToCheck_local()
    NewItem_ControlName_5_#2=@n11
   )
   n15 = For Each Loop(Exec=@n10.else, Array=Get SET_Map_IKIFK())
   n16 = Get ControlToCheck_local() -> (ControlToCheck_local=None)
   > DEBUG(NewParam=Append(
     A="don't select--control to check:  "
     B="To String (Name)"(InName=@n16)
     C="    target Control: "
     D="To String (Name)"(InName=@n15)
     E="    dest control:"
     F="To String (Name)"(InName=@n15.Array Element_destinationControl_4_#4)
    ))
   > Branch(Condition=OR Boolean(A="Equal (Name)"(A=@n15, B=@n16), B="Equal (Name)"(A=@n16, B=@n15.Array Element_destinationControl_4_#4)))
   > Return Node!(
    execute
    IKFKCheck:bool=Get loc_ikFk()
    TargetControl:Name=@n15
    DestinationControl:Name=@n15.Array Element_destinationControl_4_#4
   )
   n17 = Branch(execute=@n6.else, Condition="Equal Exactly (String)"(A=@n4, B=Spine))
   > n18 = Set IKFK_IKFKSwitchControl(IKFK_IKFKSwitchControl=Spine/IK Vis) -> (Output_Get=None)
   n19 = Set IKFK_IKFKSwitchControl(execute=@n17.else, IKFK_IKFKSwitchControl=String To Name(InString=Append!(A=@n4, B=/Ik Fk Switch) -> (ReturnValue))) -> (Output_Get=None)
   Add!(
    execute
    TargetArray:Array of STR Control Rig Name Pair Structures=Get STR_Selection_Output()
    NewItem_ControlRig_2_#1:Control Rig Object Reference=Get ControlRigToCheck_local()
    NewItem_ControlName_5_#2:Name=Get ControlToCheck_local() -> (ControlToCheck_local=None)
   ) -> (then, ReturnValue)
   Add Unique(
    TargetArray=Get STR_Selection_Output()
    NewItem_ControlRig_2_#1=Get ControlRigToCheck_local()
    NewItem_ControlName_5_#2=Get ControlToCheck_local() -> (ControlToCheck_local=None)
   )
  ]}
 NumCheck {nodes [
   n1 = NumCheck!() -> (then, SelectionInput:Array of STR Control Rig Name Pair Structures) {is=title K2Node_FunctionEntry}
   > n2 = For Each Loop!(Exec, Array:Array of STR Control Rig Name Pair Structures=@n1) -> (
    LoopBody
    Array Element_ControlRig_2_#1:Control Rig Object Reference
    Array Element_ControlName_5_#2:Name
    Array Index
    Completed
   )
   > Set ControlRigParsed(ControlRigParsed:Control Rig Object Reference=@n2) -> (Output_Get:Control Rig Object Reference)
   > n3 = Set StringCheck(StringCheck:String="To String (Name)"(InName=@n2.Array Element_ControlName_5_#2)) -> (Output_Get:String)
   n4 = Split(SourceString=@n3, InStr=" ", SearchDir=FromEnd)
   > n5 = Branch(Condition=@n4.ReturnValue)
   > n6 = Branch(Condition=IsNumeric!(SourceString:String=@n4.RightS) -> (ReturnValue:bool))
   > n7 = Branch(Condition=Get Debug_Check())
   > DEBUG(NewParam=*****NUM CHECK PASSED*****)
   > n8 = DEBUG(NewParam=Get StringCheck() -> (StringCheck:String))
   Add(
    execute=[@n8.then, @n7.else]
    TargetArray=Get STR_Chainwalk() -> (STR_Chainwalk:Array of STR Control Rig Name Pair Structures)
    NewItem_ControlRig_2_#1=Get ControlRigParsed() -> (ControlRigParsed:Control Rig Object Reference)
    NewItem_ControlName_5_#2=String To Name(InString=Get StringCheck())
   )
   n9 = Get StringCheck()
   DEBUG(execute=[@n6.else, @n5.else], NewParam=Append!(A="***Adding ", B=@n9, C:String=" to the HardWalk list from NumCheck****") -> (ReturnValue))
   n10 = String To Name(InString=@n9)
   > n11 = Branch(Condition=Contains Item!(
     TargetArray:Array of STR Control Rig Name Pair Structures=Get STR_Hardwalk() -> (STR_Hardwalk:Array of STR Control Rig Name Pair Structures)
     ItemToFind_ControlRig_2_#1:Control Rig Object Reference=Get ControlRigParsed()
     ItemToFind_ControlName_5_#2:Name=@n10
    ) -> (ReturnValue))
   Add(
    execute=@n11.else
    TargetArray=Get STR_Hardwalk()
    NewItem_ControlRig_2_#1=Get ControlRigParsed()
    NewItem_ControlName_5_#2=@n10
   )
  ]}
 OnKeyDown {nodes [
   n1 = OnKeyDown!() -> (then, MyGeometry:Geometry Structure, InKeyEvent:Key Event Structure) {is=title K2Node_FunctionEntry}
   > Set keyPressed(keyPressed:Key Structure=GetKey!(Input:"Key Event Structure (by ref)"=@n1.InKeyEvent) -> (ReturnValue:Key Structure)) -> (Output_Get:Key Structure)
   > n2 = Branch(Condition="Equal (Key)"(A:Key Structure=Get keyPressed() -> (keyPressed:Key Structure), B:Key Structure=W) -> (ReturnValue:bool=~) {is=title K2Node_PromotableOperator})
   > n3 = PickWalk_MASTER(Direction=1)
   Return Node!(execute=[@n3.then, @n5.then], ReturnValue:Event Reply Structure=Handled())
   n4 = Branch(execute=@n2.else, Condition="Equal (Key)"(A=Get keyPressed(), B=S) -> (ReturnValue=~))
   > n5 = PickWalk_MASTER(Direction=-1)
   n6 = Branch(execute=@n4.else, Condition="Equal (Key)"(A=Get keyPressed(), B=E) -> (ReturnValue=~))
   > Select_Part()
   n7 = Branch(execute=@n6.else, Condition="Equal (Key)"(A=Get keyPressed(), B=Q) -> (ReturnValue=~))
   > Select_Mirror()
   n8 = Branch(execute=@n7.else, Condition="Equal (Key)"(A=Get keyPressed(), B=Z) -> (ReturnValue=~))
   > Recall_LastInput()
   n9 = Branch(execute=@n8.else, Condition="Equal (Key)"(A=Get keyPressed(), B=A) -> (ReturnValue=~))
   > SideStep(Direction=-1)
   n10 = Branch(execute=@n9.else, Condition="Equal (Key)"(A=Get keyPressed(), B=D) -> (ReturnValue=~))
   > SideStep(Direction=1)
   n11 = Branch(execute=@n10.else, Condition="Equal (Key)"(A=Get keyPressed(), B=C) -> (ReturnValue=~))
   > SpecialShit()
   n12 = Branch(execute=@n11.else, Condition="Equal (Key)"(A=Get keyPressed(), B=Home) -> (ReturnValue=~))
   > n13 = Branch(Condition="Equal (Boolean)"(A=Get Debug_Check()) -> (ReturnValue=~))
   > Set Debug_Check(Debug_Check=true)
   > GetSelection(ClearSelection?=false) {is=title K2Node_CallFunction}
   Set Debug_Check(execute=@n13.else, Debug_Check=false)
   Return Node(execute=@n12.else, ReturnValue=Unhandled())
  ]}
 OnMouseMove {nodes [
   OnMouseMove!() -> (then, MyGeometry:Geometry Structure, MouseEvent:"Pointer Event Structure (by ref)") {is=title K2Node_FunctionEntry}
   > SetKeyboardFocus(self:Widget Object Reference)
   > Return Node(ReturnValue=Handled())
  ]}
 OnMouseMove_CheckForLeftRight {nodes [
   GetSelection(ClearSelection?=false)
   > n1 = Sequence!(execute) -> (then_0, then_1, then_2, then_3:Exec)
   > n2 = For Each Loop with Break(Array=Get Selection_Input(), Break=@n4.then)
   > n3 = Branch(Condition=OR Boolean(A="Equal (Name)"(A=@n2, B=torso_ctrl), B="Equal (Name)"(A=@n2, B=body_ctrl)))
   > SetIsEnabled(bInIsEnabled=true)
   > n4 = SetIsEnabled(bInIsEnabled=true)
   SetIsEnabled(execute=@n3.else, bInIsEnabled=false)
   > SetIsEnabled(bInIsEnabled=false)
   n5 = For Each Loop with Break(Exec=@n1.then_1, Array=Get Selection_Input())
   n6 = Get Arm_L_FK()
   n7 = Get Arm_L_IK()
   > n8 = Branch(Condition=OR Boolean(
     A=Contains Item!(TargetArray:Array of Names=@n6, ItemToFind:"Name (by ref)"=@n5) -> (ReturnValue)
     B=Contains Item(TargetArray=@n7, ItemToFind=@n5)
    ))
   > SetIsEnabled(bInIsEnabled=true)
   SetIsEnabled(execute=@n8.else, bInIsEnabled=false)
   n9 = For Each Loop with Break(Exec=@n1.then_2, Array=Get Selection_Input())
   n10 = Get Arm_R_FK()
   n11 = Get Arm_R_IK()
   > n12 = Branch(Condition=OR Boolean(A=Contains Item(TargetArray=@n10, ItemToFind=@n9), B=Contains Item(TargetArray=@n11, ItemToFind=@n9)))
   > SetIsEnabled(bInIsEnabled=true)
   SetIsEnabled(execute=@n12.else, bInIsEnabled=false)
   Return Node(execute=@n1.then_3, ReturnValue=Handled())
   Contains Item(TargetArray=Get Leg_L_IK(), ItemToFind=@n5)
   Contains Item(TargetArray=Get Leg_L_FK(), ItemToFind=@n5)
   OnMouseMove_CheckForLeftRight!() -> (then, MyGeometry:Geometry Structure, MouseEvent:"Pointer Event Structure (by ref)") {is=title K2Node_FunctionEntry}
  ]}
 PickWalk_MASTER {nodes [
   n1 = PickWalk_MASTER!() -> (then, Direction:int) {is=title K2Node_FunctionEntry}
   > Set Direction(Direction:int=@n1) -> (Output_Get:int=0)
   > GetSelection(ClearSelection?=true)
   > NumCheck!(execute, self, SelectionInput:Array of STR Control Rig Name Pair Structures=Get STR_Selection_Input()) -> (then) {is=title K2Node_CallFunction}
   > n2 = Sequence!(execute) -> (then_0, then_1, then_2)
   > Pickwalk_Chain()
   Pickwalk_Hard(execute=@n2.then_1)
   n3 = For Each Loop(Exec=@n2.then_2, Array=Get STR_Selection_Output())
   DEBUG(execute=@n3.Completed, NewParam=*****NEW SELECTION ARRAY****)
   > n4 = For Each Loop(Array=Get STR_Selection_Output())
   > DEBUG(NewParam="To String (Name)"(InName=@n4.Array Element_ControlName_5_#2))
   > SelectControl(self=@n4, InControlName=@n4.Array Element_ControlName_5_#2, bSelect=true, bSetupUndo=false)
   ClearControlSelection(execute=@n3.LoopBody, self=@n3, bSetupUndo=false)
  ]}
 Pickwalk_Chain {nodes [
   Pickwalk_Chain() {is=title K2Node_FunctionEntry}
   > DEBUG(NewParam=*****Chainwalking*****)
   > n1 = For Each Loop(Array=Get STR_Chainwalk())
   > DEBUG(NewParam="To String (Name)"(InName=@n1.Array Element_ControlName_5_#2))
   > Set ParsedControlRig_Chainwalk(ParsedControlRig_Chainwalk:Control Rig Object Reference=@n1) -> (Output_Get:Control Rig Object Reference)
   > Set parsedString(parsedString:String="To String (Name)"(InName=@n1.Array Element_ControlName_5_#2)) -> (Output_Get:String)
   n2 = Get parsedString() -> (parsedString:String)
   > n3 = Branch(Condition=OR Boolean(A=MatchesWildcard(SourceString=@n2, Wildcard=*Spine*), B=MatchesWildcard(SourceString=@n2, Wildcard=*Neck*)))
   > n4 = Set direction_local(direction_local:int=Get Direction() -> (Direction=0)) -> (Output_Get:int=0)
   n5 = Split!(execute:Exec=[@n4.then, @n17.then], SourceString=Get parsedString(), InStr=" ", SearchCase, SearchDir=FromEnd) -> (then:Exec, LeftS, RightS, ReturnValue)
   n6 = int + int(A=String To Integer(InString=@n5.RightS), B=Get direction_local() -> (direction_local:int=0))
   > n7 = Branch(Condition="integer >= integer"(A=@n6))
   n8 = "To String (Integer)"(InInt=@n6)
   > Set iteratedControl(iteratedControl:Name=String To Name(InString=Append(A=@n5, B=" ", C=@n8))) -> (Output_Get:Name=None)
   > n9 = ControlExistCheck(
    Control Rig=Get ParsedControlRig_Chainwalk() -> (ParsedControlRig_Chainwalk:Control Rig Object Reference)
    ControlName_In=Get iteratedControl() -> (iteratedControl:Name=None)
   ) -> (ControlName_Out=None)
   > n10 = Branch(Condition=@n9)
   > n11 = Set Hierarchy(Hierarchy=GetHierarchy(self=Get ParsedControlRig_Chainwalk()))
   > n12 = For Each Loop!(
    Exec
    Array:Array of Rig Element Key Structures=Get Controls!(self:Rig Hierarchy Object Reference=@n11, bTraverse:bool=true) -> (ReturnValue:Array of Rig Element Key Structures) {is=title K2Node_CallFunction}
   ) -> (LoopBody, Array Element_Type:ERigElementType Enum, Array Element_Name:Name, Array Index, Completed)
   n13 = Branch(
    execute=@n12.Completed
    Condition=Contains Item(TargetArray=Get ChainArray() -> (ChainArray:Array of Names), ItemToFind=Get iteratedControl() -> (iteratedControl=None))
   )
   > Add Unique(
    TargetArray=Get STR_Selection_Output()
    NewItem_ControlRig_2_#1=Get ParsedControlRig_Chainwalk()
    NewItem_ControlName_5_#2=Get iteratedControl() -> (iteratedControl=None)
   )
   Add Unique(
    execute=@n13.else
    TargetArray=Get STR_Hardwalk()
    NewItem_ControlRig_2_#1=Get ParsedControlRig_Chainwalk()
    NewItem_ControlName_5_#2=String To Name(InString=Get parsedString())
   )
   n14 = "To String (Name)"(InName=@n12.Array Element_Name)
   n15 = "To String (Name)"(InName=Get iteratedControl() -> (iteratedControl=None))
   Branch(
    execute=@n12.LoopBody
    Condition="Equal Exactly (String)"(
     A=Split!(SourceString=@n14, InStr=" ", SearchCase, SearchDir=FromEnd) -> (LeftS, RightS, ReturnValue)
     B=Split(SourceString=@n15, InStr=" ", SearchDir=FromEnd)
    )
   )
   > Add!(execute, TargetArray:Array of Names=Get ChainArray(), NewItem:"Name (by ref)"=@n12.Array Element_Name) -> (then, ReturnValue)
   n16 = Get parsedString()
   DEBUG(execute=[@n10.else, @n7.else], NewParam=Append(A="***Adding ", B=@n16, C=" to the Hardwalk list from failed chainwalk attempt***"))
   > Add Unique(
    TargetArray=Get STR_Hardwalk()
    NewItem_ControlRig_2_#1=Get ParsedControlRig_Chainwalk()
    NewItem_ControlName_5_#2=String To Name(InString=@n16)
   )
   n17 = Set direction_local(execute=@n3.else, direction_local=int * int(A=Get Direction() -> (Direction=0), B=-1)) -> (Output_Get=0)
  ]}
 Pickwalk_Face {nodes [
   Pickwalk_Face() {is=title K2Node_FunctionEntry}
   > n1 = Branch(Condition="Equal (Name)"(A=Get selectedControl_parsed() -> (selectedControl_parsed=None), B=mouth_ctrl))
   > n2 = Branch(Condition="Equal (Integer)"(A=Get Direction() -> (Direction=0), B=1))
   > Add(TargetArray=Get Selection_New(), NewItem=eye_mask_r_ctrl)
   > Add(TargetArray=Get Selection_New(), NewItem=eye_mask_l_ctrl)
   Add(execute=@n2.else, TargetArray=Get Selection_New(), NewItem=head_ctrl)
   n3 = Branch(execute=@n1.else, Condition="Equal (Name)"(A=Get selectedControl_parsed() -> (selectedControl_parsed=None), B=eye_mask_r_ctrl))
   > n4 = Branch(Condition="Equal (Integer)"(A=Get Direction() -> (Direction=0), B=1))
   > Add(TargetArray=Get Selection_New(), NewItem=eye_r_ctrl)
   > Add(TargetArray=Get Selection_New(), NewItem=eye_l_ctrl)
   Add(execute=@n4.else, TargetArray=Get Selection_New(), NewItem=mouth_ctrl)
   n5 = Branch(execute=@n3.else, Condition="Equal (Name)"(A=Get selectedControl_parsed() -> (selectedControl_parsed=None), B=eye_r_ctrl))
   > n6 = Branch(Condition="Equal (Integer)"(A=Get Direction() -> (Direction=0), B=1))
   > Add(TargetArray=Get Selection_New(), NewItem=eyelash_r_ctrl)
   > Add(TargetArray=Get Selection_New(), NewItem=eyelash_l_ctrl)
   Add(execute=@n6.else, TargetArray=Get Selection_New(), NewItem=eye_mask_r_ctrl)
   > Add(TargetArray=Get Selection_New(), NewItem=eye_mask_l_ctrl)
   n7 = Branch(execute=@n5.else, Condition="Equal (Name)"(A=Get selectedControl_parsed() -> (selectedControl_parsed=None), B=eyelash_r_ctrl))
   > n8 = Branch(Condition="Equal (Integer)"(A=Get Direction() -> (Direction=0), B=1))
   > Add(TargetArray=Get Selection_New(), NewItem=brow_r_ctrl)
   > Add(TargetArray=Get Selection_New(), NewItem=brow_l_ctrl)
   Add(execute=@n8.else, TargetArray=Get Selection_New(), NewItem=eye_r_ctrl)
   > Add(TargetArray=Get Selection_New(), NewItem=eye_l_ctrl)
   n9 = Branch(execute=@n7.else, Condition="Equal (Name)"(A=Get selectedControl_parsed() -> (selectedControl_parsed=None), B=brow_r_ctrl))
   > n10 = Branch(Condition="Equal (Integer)"(A=Get Direction() -> (Direction=0), B=1))
   > Add(TargetArray=Get Selection_New(), NewItem=brow_r_ctrl)
   > Add(TargetArray=Get Selection_New(), NewItem=brow_l_ctrl)
   Add(execute=@n10.else, TargetArray=Get Selection_New(), NewItem=eyelash_r_ctrl)
   > Add(TargetArray=Get Selection_New(), NewItem=eyelash_l_ctrl)
   Return Node!(execute=@n9.else, faceCtrl:bool=false)
  ]}
 Pickwalk_Hard {nodes [
   Pickwalk_Hard() {is=title K2Node_FunctionEntry}
   > DEBUG(NewParam=****HardWalking these Controls****)
   n1 = Get STR_Hardwalk()
   > n2 = For Each Loop!(Exec, Array:Array of STR Control Rig Name Pair Structures=@n1) -> (
    LoopBody
    Array Element_ControlRig_2_#1:Control Rig Object Reference
    Array Element_ControlName_5_#2:Name
    Array Index
    Completed
   )
   n3 = For Each Loop(Exec=@n2.Completed, Array=@n1)
   > Set ParsedControlRig_hardwalk(ParsedControlRig_hardwalk:Control Rig Object Reference=@n3) -> (Output_Get:Control Rig Object Reference)
   > Set ParsedControlName(ParsedControlName:Name=@n3.Array Element_ControlName_5_#2) -> (Output_Get:Name=None)
   > n4 = For Each Loop with Break!(Exec, Array:Array of STR Pick Walk SET Structures=Get SET_Map() -> (SET_Map:Array of STR Pick Walk SET Structures), Break) -> (
    LoopBody
    Array Element_PreviousControl_2_#5:Name
    Array Element_ControlSET_7_#6:Array of Names
    Array Element_NextControl_6_#7:Name
    Array Index
    Completed
   )
   n5 = Branch(execute=@n4.Completed, Condition=Get SetContains_Check() -> (SetContains_Check:bool))
   > DEBUG(NewParam=*****PARSED CONTROL SET*****)
   > n6 = For Each Loop!(Exec, Array:Array of Names=Get ParsedControlSET() -> (ParsedControlSET:Array of Names)) -> (LoopBody, Array Element:Name, Array Index, Completed)
   n7 = For Each Loop(Exec=[@n6.Completed, @n19.Completed], Array=Get ParsedControlSET())
   n8 = Set localIndex(execute=@n7.Completed, localIndex:int=int + int(A=Get Direction() -> (Direction=0), B=Get localIndex() -> (localIndex:int=0))) -> (Output_Get:int=0)
   > DEBUG(NewParam=Append!(A="***New Index: ", B="To String (Integer)"(InInt=@n8)) -> (ReturnValue))
   n9 = Length(TargetArray:Array of Names=Get ParsedControlSET())
   > DEBUG(NewParam=Append(A="***array length: ", B="To String (Integer)"(InInt=@n9)))
   n10 = Get localIndex() -> (localIndex=0)
   > n11 = Branch(Condition="integer > integer"(A=int + int(A=@n10), B=@n9))
   > DEBUG(NewParam="Selecting Previous: ")
   > DEBUG(NewParam="To String (Name)"(InName=Get ParsedPreviousControl() -> (ParsedPreviousControl:Name=None)))
   > n12 = Branch(Condition=Contains Item(TargetArray=Get IKCheckList() -> (IKCheckList:Array of Names), ItemToFind=Get ParsedPreviousControl() -> (ParsedPreviousControl=None)))
   > IKFK_Check!(
    execute
    self
    control to check:Name=Get ParsedPreviousControl() -> (ParsedPreviousControl=None)
    controlRigToCheck:Control Rig Object Reference=Get ParsedControlRig_hardwalk() -> (ParsedControlRig_hardwalk:Control Rig Object Reference)
    SelectControl:bool=true
   ) -> (then, IKFKCheck:bool, TargetControl:Name=None, DestinationControl:Name=None) {is=title K2Node_CallFunction}
   Add Unique(
    execute=@n12.else
    TargetArray=Get STR_Selection_Output()
    NewItem_ControlRig_2_#1=Get ParsedControlRig_hardwalk()
    NewItem_ControlName_5_#2=Get ParsedPreviousControl() -> (ParsedPreviousControl=None)
   )
   n13 = Branch(execute=@n11.else, Condition="integer < integer"(A=Get localIndex() -> (localIndex=0)))
   > DEBUG(NewParam="Selecting Previous: ")
   > DEBUG(NewParam="To String (Name)"(InName=Get ParsedNextControl() -> (ParsedNextControl:Name=None)))
   > n14 = Branch(Condition=Contains Item(TargetArray=Get IKCheckList(), ItemToFind=Get ParsedNextControl() -> (ParsedNextControl=None)))
   > IKFK_Check(control to check=Get ParsedNextControl() -> (ParsedNextControl=None), controlRigToCheck=Get ParsedControlRig_hardwalk(), SelectControl=true) -> (TargetControl=None, DestinationControl=None)
   Add Unique(
    execute=@n14.else
    TargetArray=Get STR_Selection_Output()
    NewItem_ControlRig_2_#1=Get ParsedControlRig_hardwalk()
    NewItem_ControlName_5_#2=Get ParsedNextControl() -> (ParsedNextControl=None)
   )
   DEBUG(execute=@n13.else, NewParam="Selecting next in SET: ")
   n15 = "Get (a copy)"(Array:Array of Names=Get ParsedControlSET(), Dimension 1=Get localIndex() -> (localIndex=0)) -> (Output:Name)
   > DEBUG(NewParam="To String (Name)"(InName=@n15))
   > n16 = Branch(Condition=Contains Item(TargetArray=Get IKCheckList(), ItemToFind=@n15))
   > IKFK_Check(control to check=@n15, controlRigToCheck=Get ParsedControlRig_hardwalk(), SelectControl=true) -> (TargetControl=None, DestinationControl=None)
   Add Unique(
    execute=@n16.else
    TargetArray=Get STR_Selection_Output()
    NewItem_ControlRig_2_#1=Get ParsedControlRig_hardwalk()
    NewItem_ControlName_5_#2=@n15
   )
   Branch(execute=@n7.LoopBody, Condition="Equal (Name)"(A=@n7, B=Get ParsedControlName() -> (ParsedControlName:Name=None)))
   > n17 = Set localIndex(localIndex=@n7.Array Index) -> (Output_Get=0)
   > DEBUG(NewParam=Append(A="***Array index: ", B="To String (Integer)"(InInt=@n17)))
   DEBUG(execute=@n6.LoopBody, NewParam="To String (Name)"(InName=@n6))
   n18 = For Each Loop!(Exec=@n5.else, Array:Array of STR Pick Walk SET Structures=Get SET_Map_Hand() -> (SET_Map_Hand:Array of STR Pick Walk SET Structures)) -> (
    LoopBody
    Array Element_PreviousControl_2_#5:Name
    Array Element_ControlSET_7_#6:Array of Names
    Array Element_NextControl_6_#7:Name
    Array Index
    Completed
   )
   DEBUG(execute=@n18.Completed, NewParam="***  FingerSet Found  ***")
   > n19 = For Each Loop!(Exec, Array:Array of Names=Get ParsedControlSET()) -> (LoopBody, Array Element:Name, Array Index, Completed)
   > DEBUG(NewParam="To String (Name)"(InName=@n19))
   n20 = Get_L_or_R!(
    execute=@n18.LoopBody
    self
    in_sourceControl:Name=Get ParsedControlName() -> (ParsedControlName=None)
    in_array:Array of Names=@n18.Array Element_ControlSET_7_#6
   ) -> (then, out_LorR:String, out_array:Array of Names) {is=title K2Node_CallFunction}
   > Branch(Condition=Contains Item(TargetArray=@n20.out_array, ItemToFind=Get ParsedControlName() -> (ParsedControlName=None)))
   > Set SetContains_Check(SetContains_Check:bool=true) -> (Output_Get:bool)
   n21 = "To String (Name)"(InName=@n18)
   > Set ParsedPreviousControl(ParsedPreviousControl:Name=String To Name(InString=Replace(SourceString=@n21, From=Q, To=@n20))) -> (Output_Get:Name=None)
   > Set ParsedControlSET(ParsedControlSET:Array of Names=@n20.out_array) -> (Output_Get:Array of Names)
   n22 = "To String (Name)"(InName=@n18.Array Element_NextControl_6_#7)
   > Set ParsedNextControl(ParsedNextControl:Name=String To Name(InString=Replace(SourceString=@n22, From=Q, To=@n20))) -> (Output_Get:Name=None)
   Branch(
    execute=@n4.LoopBody
    Condition=Contains Item(TargetArray=@n4.Array Element_ControlSET_7_#6, ItemToFind=Get ParsedControlName() -> (ParsedControlName=None))
   )
   > Set SetContains_Check(SetContains_Check=true)
   > Set ParsedPreviousControl(ParsedPreviousControl=@n4) -> (Output_Get=None)
   > Set ParsedControlSET(ParsedControlSET=@n4.Array Element_ControlSET_7_#6)
   > Set ParsedNextControl(ParsedNextControl=@n4.Array Element_NextControl_6_#7) -> (Output_Get=None)
   DEBUG(execute=@n2.LoopBody, NewParam="To String (Name)"(InName=@n2.Array Element_ControlName_5_#2))
  ]}
 Pickwalk_Parse {nodes [
   n1 = Pickwalk_Parse!() -> (then, selectedControl:Name, bodySet:Array of Names) {is=title K2Node_FunctionEntry}
   > n2 = For Each Loop with Break!(Exec, Array:Array of Names=@n1.bodySet, Break=@n9.then) -> (LoopBody, Array Element:Name, Array Index, Completed)
   n3 = Branch(execute=@n2.Completed, Condition=Get nameCheck() -> (nameCheck:bool))
   n4 = Get Direction() -> (Direction=0)
   n5 = Get currIndex() -> (currIndex:int=0)
   > Add(
    TargetArray=Get Selection_New()
    NewItem="Get (a copy)"(Array=Get bodySet() -> (bodySet:Array of Names), Dimension 1=int + int(A=@n5, B=@n4))
   )
   > Return Node!(execute, False:bool=true)
   Return Node(execute=@n3.else, False=false)
   n6 = Set currIndex(execute=@n2.LoopBody, currIndex:int=@n2.Array Index) -> (Output_Get:int=0)
   n7 = Length(TargetArray=Get bodySet())
   > Branch(Condition="In Range (Integer)"(Value=@n6, Min=1, Max=int - int(A=@n7, B=2), InclusiveMin=true, InclusiveMax=true))
   n8 = "Equal (Name)"(A=@n2, B=Get selectedControl() -> (selectedControl:Name=None))
   > Set nameCheck(nameCheck:bool=@n8) -> (Output_Get:bool)
   > n9 = Branch(Condition=@n8)
   Return Node(execute=@n17.else, False=Get nameCheck())
   Return Node(execute=[@n10.then, @n13.then, @n14.then], False=Get nameCheck())
   n10 = Add(execute=@n18.then, TargetArray=Get Selection_New())
   n11 = For Each Loop with Break(Break=@n20.then)
   n12 = int + int(A=@n11.Array Index, B=Get Direction() -> (Direction=0))
   n13 = Add(execute=@n16.else, TargetArray=Get Selection_New(), NewItem="Get (a copy)"(Dimension 1=@n12))
   n14 = Add(execute=@n16.then, TargetArray=Get Selection_New())
   n15 = Length(TargetArray:Array of Wildcards)
   n16 = Branch(execute=@n18.else, Condition="integer < integer"(A=@n12))
   n17 = Branch(execute=@n11.Completed, Condition=Get nameCheck())
   > n18 = Branch(Condition="integer >= integer"(A=@n12, B=@n15))
   n19 = "Equal (Name)"(A=@n11)
   n20 = Branch(execute=@n21.then, Condition=@n19)
   n21 = Set nameCheck(execute=@n11.LoopBody, nameCheck=@n19)
  ]}
 RESET Vars {nodes [
   RESET Vars() {is=title K2Node_FunctionEntry}
   > n1 = Sequence!(execute) -> (then_0, then_1)
   > Clear(TargetArray:Array of Names=Get IKCheckList())
   > Clear(TargetArray:Array of STR Control Rig Name Pair Structures=Get STR_Selection_Input())
   > Clear(TargetArray=Get STR_Hardwalk())
   > Clear(TargetArray=Get STR_Chainwalk())
   > Clear(TargetArray=Get STR_Selection_Output())
   n2 = For Each Loop!(Exec=@n1.then_1, Array:Array of STR Pickwalk Map IKFK Structures=Get SET_Map_IKIFK()) -> (
    LoopBody
    Array Element_targetControl_2_#3:Name
    Array Element_destinationControl_4_#4:Name
    Array Index
    Completed
   )
   > Add(TargetArray=Get IKCheckList(), NewItem=@n2)
  ]}
 Recall_LastInput {nodes [
   Recall_LastInput() {is=title K2Node_FunctionEntry}
   n1 = Get STR_Selection_Input()
   > n2 = Branch(Condition="integer > integer"(A=Length(TargetArray:Array of STR Control Rig Name Pair Structures=@n1)))
   > n3 = For Each Loop!(Exec, Array:Array of STR Control Rig Name Pair Structures=Get STR_Selection_Input()) -> (
    LoopBody
    Array Element_ControlRig_2_#1:Control Rig Object Reference
    Array Element_ControlName_5_#2:Name
    Array Index
    Completed
   )
   n4 = For Each Loop(Exec=@n3.Completed, Array=Get STR_Selection_Input())
   > DEBUG(NewParam="To String (Name)"(InName=@n4.Array Element_ControlName_5_#2))
   > SelectControl(self=@n4, InControlName=@n4.Array Element_ControlName_5_#2, bSelect=true, bSetupUndo=false)
   ClearControlSelection(execute=@n3.LoopBody, self=@n3, bSetupUndo=false)
   n5 = GetFocusedLevelSequence(execute=@n2.else)
   > n6 = Set CurrLevelSequence(CurrLevelSequence=@n5)
   > n7 = GetControlRigs(LevelSequence=@n6)
   > n8 = For Each Loop!(Exec, Array:Array of Control Rig Sequencer Binding Proxy Structures=@n7) -> (LoopBody, Array Element:Control Rig Sequencer Binding Proxy Structure, Array Index, Completed)
   n9 = Get Control Rig() -> (Control Rig:Control Rig Object Reference)
   ClearControlSelection(execute=@n8.Completed, self=@n9, bSetupUndo=false)
   > SelectControl(self=@n9, InControlName=String To Name(InString=Body/Body), bSelect=true, bSetupUndo=false)
   n10 = Break Control Rig Sequencer Binding Proxy(ControlRigSequencerBindingProxy=@n8)
   Branch(
    execute=@n8.LoopBody
    Condition=Contains(SearchIn="To String (Object)"!(InObj:Object Reference=@n10.ControlRig) -> (ReturnValue:String), Substring=Biped)
   )
   > Set Control Rig(Control Rig=@n10.ControlRig)
  ]}
 Select_Mirror {nodes [
   Select_Mirror() {is=title K2Node_FunctionEntry}
   > n1 = GetSelection(ClearSelection?=false)
   > n2 = For Each Loop!(Exec, Array:Array of STR Control Rig Name Pair Structures=@n1) -> (
    LoopBody
    Array Element_ControlRig_2_#1:Control Rig Object Reference
    Array Element_ControlName_5_#2:Name
    Array Index
    Completed
   )
   Select_STR_Output(execute=@n2.Completed)
   n3 = Set MirrorControl_input(execute=@n2.LoopBody, MirrorControl_input:Name=@n2.Array Element_ControlName_5_#2) -> (Output_Get:Name=None)
   > Set Mirror_CR(Mirror_CR:Control Rig Object Reference=@n2) -> (Output_Get:Control Rig Object Reference)
   n4 = Split(SourceString="To String (Name)"(InName=@n3), InStr=" ", SearchDir=FromStart)
   n5 = Split(SourceString=@n4.RightS, InStr=/, SearchDir=FromStart)
   > n6 = Branch(Condition="Equal, Case Insensitive (String)"(A=@n5, B=R))
   > n7 = Set MirrorControl_target(MirrorControl_target:Name=String To Name(InString=Append!(A=@n4, B=" L/", C:String=@n5.RightS) -> (ReturnValue))) -> (Output_Get:Name=None)
   n8 = Get Mirror_CR() -> (Mirror_CR:Control Rig Object Reference)
   n9 = IKFK_Check(
    execute=[@n7.then, @n21.then]
    control to check=Get MirrorControl_target() -> (MirrorControl_target:Name=None)
    controlRigToCheck=@n8
    SelectControl=false
   ) -> (TargetControl=None, DestinationControl=None)
   > Set MirrorControl_target_IKFK(MirrorControl_target_IKFK:bool=@n9) -> (Output_Get:bool)
   > n10 = IKFK_Check(control to check=Get MirrorControl_input() -> (MirrorControl_input:Name=None), controlRigToCheck=@n8, SelectControl=false) -> (TargetControl=None, DestinationControl=None)
   > Set MirrorControl_input_IKFK(MirrorControl_input_IKFK:bool=@n10) -> (Output_Get:bool)
   > n11 = Branch(Condition="Equal (Boolean)"(A=Get MirrorControl_target_IKFK() -> (MirrorControl_target_IKFK:bool), B=Get MirrorControl_input_IKFK() -> (MirrorControl_input_IKFK:bool)) -> (ReturnValue=~))
   > Add Unique(
    TargetArray=Get STR_Selection_Output()
    NewItem_ControlRig_2_#1=Get Mirror_CR()
    NewItem_ControlName_5_#2=Get MirrorControl_target() -> (MirrorControl_target=None)
   )
   n12 = Branch(execute=@n11.else, Condition=Get MirrorControl_target_IKFK())
   n13 = Split(SourceString="To String (Name)"(InName=Get MirrorControl_target() -> (MirrorControl_target=None)), InStr=/, SearchDir=FromStart)
   > n14 = Branch(Condition="Equal, Case Insensitive (String)"(A=@n13.RightS, B=FK 2))
   > n15 = Set MirrorControl_target(MirrorControl_target=String To Name(InString=Append!(A=@n13, B=/IK) -> (ReturnValue))) -> (Output_Get=None)
   Add Unique(
    execute=[@n15.then, @n16.then]
    TargetArray=Get STR_Selection_Output()
    NewItem_ControlRig_2_#1=Get Mirror_CR()
    NewItem_ControlName_5_#2=Get MirrorControl_target() -> (MirrorControl_target=None)
   )
   n16 = Set MirrorControl_target(execute=@n14.else, MirrorControl_target=String To Name(InString=Append(A=@n13, B=/PV))) -> (Output_Get=None)
   n17 = Split(SourceString="To String (Name)"(InName=Get MirrorControl_target() -> (MirrorControl_target=None)), InStr=/, SearchDir=FromStart)
   n18 = Branch(execute=@n12.else, Condition="Equal, Case Insensitive (String)"(A=@n17.RightS, B=IK))
   > n19 = Set MirrorControl_target(MirrorControl_target=String To Name(InString=Append(A=@n17, B=/FK 2))) -> (Output_Get=None)
   Add Unique(
    execute=[@n20.then, @n19.then]
    TargetArray=Get STR_Selection_Output()
    NewItem_ControlRig_2_#1=Get Mirror_CR()
    NewItem_ControlName_5_#2=Get MirrorControl_target() -> (MirrorControl_target=None)
   )
   n20 = Set MirrorControl_target(execute=@n18.else, MirrorControl_target=String To Name(InString=Append(A=@n17, B=/FK 1))) -> (Output_Get=None)
   n21 = Set MirrorControl_target(execute=@n6.else, MirrorControl_target=String To Name(InString=Append!(A=@n4, B=" R/", C:String=@n5.RightS) -> (ReturnValue))) -> (Output_Get=None)
  ]}
 Select_Part {nodes [
   Select_Part() {is=title K2Node_FunctionEntry}
   > n1 = GetSelection(ClearSelection?=false)
   > n2 = For Each Loop(Array=@n1)
   > Set ControlRig_In_SelPart(ControlRig_In_SelPart:Control Rig Object Reference=@n2) -> (Output_Get:Control Rig Object Reference)
   > Set ControlName_In_SelPart(ControlName_In_SelPart:Name=@n2.Array Element_ControlName_5_#2) -> (Output_Get:Name=None)
   > n3 = For Each Loop with Break!(Exec, Array:Array of STR Pick Walk SET Structures=Get SET_Map(), Break=@n30.then) -> (
    LoopBody
    Array Element_PreviousControl_2_#5:Name
    Array Element_ControlSET_7_#6:Array of Names
    Array Element_NextControl_6_#7:Name
    Array Index
    Completed
   )
   n4 = Branch(execute=@n3.Completed, Condition=Get SetContainsName() -> (SetContainsName:bool))
   > n5 = For Each Loop with Break!(Exec, Array:Array of Names=Get ParsedControlSET_SelPart() -> (ParsedControlSET_SelPart:Array of Names), Break) -> (LoopBody, Array Element:Name, Array Index, Completed)
   n6 = Branch(execute=@n5.Completed, Condition=Get IKFound() -> (IKFound:bool))
   > n7 = IKFK_Check(
    control to check=Get IKFKCheckList_ControlFound() -> (IKFKCheckList_ControlFound:Name=None)
    controlRigToCheck=Get ControlRig_In_SelPart() -> (ControlRig_In_SelPart:Control Rig Object Reference)
    SelectControl=false
   ) -> (TargetControl=None, DestinationControl=None)
   > n8 = Branch(Condition=@n7)
   > n9 = For Each Loop with Break!(Exec, Array:Array of STR Pick Walk SET Structures=Get SET_Map(), Break) -> (
    LoopBody
    Array Element_PreviousControl_2_#5:Name
    Array Element_ControlSET_7_#6:Array of Names
    Array Element_NextControl_6_#7:Name
    Array Index
    Completed
   )
   > Branch(Condition=Contains Item(TargetArray=@n9.Array Element_ControlSET_7_#6, ItemToFind=@n7.DestinationControl))
   > n10 = For Each Loop!(Exec, Array:Array of Names=@n9.Array Element_ControlSET_7_#6) -> (LoopBody, Array Element:Name, Array Index, Completed)
   > SelectControl(self=Get ControlRig_In_SelPart(), InControlName=@n10, bSelect=true, bSetupUndo=false)
   n11 = For Each Loop with Break(Exec=@n8.else, Array=Get SET_Map())
   > DEBUG(NewParam="To String (Name)"(InName=@n7.TargetControl))
   > Branch(Condition=Contains Item(TargetArray=@n11.Array Element_ControlSET_7_#6, ItemToFind=@n7.TargetControl))
   > n12 = For Each Loop(Array=@n11.Array Element_ControlSET_7_#6)
   > SelectControl(self=Get ControlRig_In_SelPart(), InControlName=@n12, bSelect=true, bSetupUndo=false)
   n13 = For Each Loop(Exec=@n6.else, Array=Get ParsedControlSET_SelPart())
   > SelectControl(self=Get ControlRig_In_SelPart(), InControlName=@n13, bSelect=true, bSetupUndo=false)
   n14 = Branch(execute=@n5.LoopBody, Condition=Contains Item(TargetArray=Get IKCheckList(), ItemToFind=@n5))
   > Set IKFound(IKFound:bool=true) -> (Output_Get:bool)
   > Set IKFKCheckList_ControlFound(IKFKCheckList_ControlFound:Name=@n5) -> (Output_Get:Name=None)
   n15 = Split(SourceString="To String (Name)"(InName=@n5), InStr=" ", SearchDir=FromStart)
   Branch(execute=@n14.else, Condition="Equal, Case Insensitive (String)"(A=@n15, B=Clavicle))
   > Set IKFound(IKFound=true)
   > Set IKFKCheckList_ControlFound(IKFKCheckList_ControlFound=String To Name(InString=Append(A="Arm ", B=@n15.RightS, C=" 0"))) -> (Output_Get=None)
   n16 = Get ControlName_In_SelPart() -> (ControlName_In_SelPart:Name=None)
   n17 = Split(SourceString=Split(SourceString="To String (Name)"(InName=@n16), InStr=/, SearchDir=FromStart), InStr=" ", SearchDir=FromStart)
   n18 = Branch(
    execute=@n4.else
    Condition=Contains Item(TargetArray=Get FingerNames() -> (FingerNames:Array of Names), ItemToFind=String To Name(InString=@n17))
   )
   > n19 = For Each Loop(Array=Get FingerNames())
   n20 = For Each Loop(Exec=@n19.Completed, Array=Get FingerControls() -> (FingerControls:Array of Names))
   > SelectControl(self=Get ControlRig_In_SelPart(), InControlName=@n20, bSelect=true, bSetupUndo=false)
   Branch(execute=@n19.LoopBody, Condition="integer > integer"(A=@n19.Array Index))
   n21 = "To String (Name)"(InName=@n19)
   n22 = Get ControlName_In_SelPart() -> (ControlName_In_SelPart=None)
   n23 = Split(SourceString=Split(SourceString="To String (Name)"(InName=@n22), InStr=/, SearchDir=FromStart), InStr=" ", SearchDir=FromStart)
   > Add Unique!(
    execute
    TargetArray:Array of Names=Get FingerControls()
    NewItem:"Name (by ref)"=String To Name(InString=Append!(A=@n21, B=" ", C=@n23.RightS, D:String=/FK 0) -> (ReturnValue))
   ) -> (then, ReturnValue)
   > Add Unique(TargetArray=Get FingerControls(), NewItem=String To Name(InString=Append(A=@n21, B=" ", C=@n23.RightS, D=/FK 1)))
   > Add Unique(TargetArray=Get FingerControls(), NewItem=String To Name(InString=Append(A=@n21, B=" ", C=@n23.RightS, D=/FK 2)))
   n24 = Append!(A="Meta ", B=@n23.RightS, C=/, D=@n21, E:String=" Metacarpal ", F:String=@n23.RightS) -> (ReturnValue)
   > Add Unique(TargetArray=Get FingerControls(), NewItem=String To Name(InString=@n24))
   > DEBUG(NewParam=Append!(A="** finger selected check ***    ", B=@n24) -> (ReturnValue))
   n25 = Split(SourceString="To String (Name)"(InName=Get ControlName_In_SelPart() -> (ControlName_In_SelPart=None)), InStr=" ", SearchDir=FromEnd)
   DEBUG(
    execute=@n18.else
    NewParam=Append!(A="***** FALLBACK, attempting to select anything that starts with \"", B=@n25, C:String="\" *****") -> (ReturnValue)
   )
   > n26 = Set Hierarchy(Hierarchy=GetHierarchy(self=Get ControlRig_In_SelPart()))
   > n27 = For Each Loop!(Exec, Array:Array of Rig Element Key Structures=Get Controls(self=@n26, bTraverse=true)) -> (LoopBody, Array Element_Type:ERigElementType Enum, Array Element_Name:Name, Array Index, Completed)
   n28 = Split(SourceString="To String (Name)"(InName=@n27.Array Element_Name), InStr=" ", SearchDir=FromEnd)
   > Branch(Condition=@n28.ReturnValue)
   > Branch(Condition="Equal Exactly (String)"(A=@n28, B=@n25))
   > SelectControl(self=Get ControlRig_In_SelPart(), InControlName=@n27.Array Element_Name, bSelect=true, bSetupUndo=false)
   Set ParsedControlSET_SelPart(execute=@n3.LoopBody, ParsedControlSET_SelPart:Array of Names=@n3.Array Element_ControlSET_7_#6) -> (Output_Get:Array of Names)
   > n29 = Branch(Condition=Contains Item(
     TargetArray=@n3.Array Element_ControlSET_7_#6
     ItemToFind=Get ControlName_In_SelPart() -> (ControlName_In_SelPart=None)
    ))
   > n30 = Set SetContainsName(SetContainsName:bool=true) -> (Output_Get:bool)
   Set SetContainsName(execute=@n29.else, SetContainsName=false)
  ]}
 Select_STR_Output {nodes [
   Select_STR_Output() {is=title K2Node_FunctionEntry}
   > n1 = For Each Loop!(Exec, Array:Array of STR Control Rig Name Pair Structures=Get STR_Selection_Output()) -> (
    LoopBody
    Array Element_ControlRig_2_#1:Control Rig Object Reference
    Array Element_ControlName_5_#2:Name
    Array Index
    Completed
   )
   DEBUG(execute=@n1.Completed, NewParam=*****NEW SELECTION ARRAY****)
   > n2 = For Each Loop(Array=Get STR_Selection_Output())
   > DEBUG(NewParam="To String (Name)"(InName=@n2.Array Element_ControlName_5_#2))
   > SelectControl(self=@n2, InControlName=@n2.Array Element_ControlName_5_#2, bSelect=true, bSetupUndo=false)
   ClearControlSelection(execute=@n1.LoopBody, self=@n1, bSetupUndo=false)
  ]}
 SideStep {nodes [
   n1 = SideStep!() -> (then, Direction:int) {is=title K2Node_FunctionEntry}
   > Set Direction(Direction=@n1) -> (Output_Get=0)
   > n2 = GetSelection(ClearSelection?=true)
   > n3 = For Each Loop(Array=@n2)
   n4 = For Each Loop!(Exec=@n3.Completed, Array:Array of Names=Get FingerNames()) -> (LoopBody, Array Element:Name, Array Index, Completed)
   n5 = For Each Loop!(
    Exec=@n4.Completed
    Array:Array of STR Control Rig Name Pair Structures=Get SideStep_Remainder() -> (SideStep_Remainder:Array of STR Control Rig Name Pair Structures)
   ) -> (
    LoopBody
    Array Element_ControlRig_2_#1:Control Rig Object Reference
    Array Element_ControlName_5_#2:Name
    Array Index
    Completed
   )
   n6 = Sequence!(execute=@n5.Completed) -> (then_0, then_1, then_2:Exec)
   > n7 = For Each Loop(Array=Get SideStep_Remainder())
   DEBUG(execute=@n7.Completed, NewParam="=== Remainder to Side Step ====")
   > n8 = For Each Loop(Array=Get SideStep_Remainder())
   > DEBUG(NewParam="To String (Name)"(InName=@n8.Array Element_ControlName_5_#2))
   n9 = For Each Loop!(Exec=@n7.LoopBody, Array:Array of Names=Get SideStep_Remainder_RemoveList() -> (SideStep_Remainder_RemoveList:Array of Names)) -> (LoopBody, Array Element:Name, Array Index, Completed)
   > Branch(Condition="Equal (Name)"(A=@n9, B=@n7.Array Element_ControlName_5_#2))
   > Set NewLocalVar(NewLocalVar:bool=true) -> (Output_Get:bool)
   n10 = Branch(execute=@n6.then_1, Condition=Get NewLocalVar() -> (NewLocalVar:bool))
   ClearControlSelection(execute=@n10.else, self=Get SideStep_CR() -> (SideStep_CR:Control Rig Object Reference), bSetupUndo=false)
   > n11 = For Each Loop!(Exec, Array:Array of STR Control Rig Name Pair Structures=Get SideStep_Remainder()) -> (
    LoopBody
    Array Element_ControlRig_2_#1:Control Rig Object Reference
    Array Element_ControlName_5_#2:Name
    Array Index
    Completed
   )
   > SelectControl(self=@n11, InControlName=@n11.Array Element_ControlName_5_#2, bSelect=true, bSetupUndo=false)
   > Select_Mirror() {is=title K2Node_CallFunction}
   Select_STR_Output(execute=@n6.then_2) {is=title K2Node_CallFunction}
   n12 = For Each Loop!(Exec=@n5.LoopBody, Array:Array of Strings=Get SideStep_FingerNames() -> (SideStep_FingerNames:Array of Strings)) -> (LoopBody, Array Element:String, Array Index, Completed)
   n13 = "To String (Name)"(InName=@n5.Array Element_ControlName_5_#2)
   > Branch(Condition=Contains(SearchIn=@n13, Substring=@n12))
   > n14 = Set SideStep_ControlName_OG(SideStep_ControlName_OG:String=@n13) -> (Output_Get:String)
   > PrintString(InString=@n14)
   n15 = Split(SourceString=@n13, InStr=" ", SearchDir=FromStart)
   > Set SideStep_NamePrefix(SideStep_NamePrefix:String=@n15) -> (Output_Get:String)
   > Set SideStep_NameTrunk(SideStep_NameTrunk:String=@n15.RightS) -> (Output_Get:String)
   > Set SideStep_LorR(SideStep_LorR:String=Split(SourceString=@n15.RightS, InStr=/, SearchDir=FromStart)) -> (Output_Get:String)
   > n16 = Branch(Condition="Equal Exactly (String)"(A=Get SideStep_NamePrefix() -> (SideStep_NamePrefix:String), B=Meta))
   n17 = Get SideStep_ControlName_OG() -> (SideStep_ControlName_OG:String)
   > Set SideStep_NamePrefix(SideStep_NamePrefix=Split(SourceString=Split(SourceString=@n17, InStr=/, SearchDir=FromStart).RightS, InStr=" ", SearchDir=FromStart))
   > n18 = Set SideStep_isMetaCtrl(SideStep_isMetaCtrl:bool=true) -> (Output_Get:bool)
   n19 = For Each Loop(Exec=[@n16.else, @n18.then], Array=Get SideStep_FingerNames())
   n20 = Branch(execute=@n19.Completed, Condition="Equal Exactly (String)"(A=Get SideStep_LorR() -> (SideStep_LorR:String), B=R))
   > n21 = Set SideStep_Direction(SideStep_Direction:int=int * int(A=Get Direction() -> (Direction=0), B=-1)) -> (Output_Get:int=0)
   n22 = Set SideStep_FingerName_Index(
    execute=[@n21.then, @n35.then]
    SideStep_FingerName_Index:int=int + int(A=Get SideStep_Direction() -> (SideStep_Direction:int=0), B=Get SideStep_FingerName_Index() -> (SideStep_FingerName_Index:int=0))
   ) -> (Output_Get:int=0)
   > n23 = Branch(Condition="integer > integer"(A=@n22, B=5))
   > n24 = Branch(Condition=Get SideStep_isMetaCtrl() -> (SideStep_isMetaCtrl:bool))
   > n25 = Set SideStep_FingerName_Index(SideStep_FingerName_Index=2) -> (Output_Get=0)
   n26 = Get SideStep_FingerNames()
   n27 = Get SideStep_FingerName_Index() -> (SideStep_FingerName_Index=0)
   n28 = Set SideStep_OutControl(
    execute=[@n33.else, @n32.then, @n25.then, @n34.then, @n29.then]
    SideStep_OutControl:String=Replace(
     SourceString=Get SideStep_ControlName_OG()
     From=Get SideStep_NamePrefix()
     To="Get (a copy)"(Array:Array of Strings=@n26, Dimension 1=@n27) -> (Output:String)
    )
   ) -> (Output_Get:String)
   > Add Unique!(
    execute
    TargetArray:Array of STR Control Rig Name Pair Structures=Get STR_Selection_Output()
    NewItem_ControlRig_2_#1:Control Rig Object Reference=Get SideStep_CR()
    NewItem_ControlName_5_#2:Name=String To Name(InString=@n28)
   ) -> (then, ReturnValue)
   > Add Unique!(
    execute
    TargetArray:Array of Names=Get SideStep_Remainder_RemoveList()
    NewItem:"Name (by ref)"=String To Name(InString=Get SideStep_ControlName_OG())
   ) -> (then, ReturnValue)
   n29 = Set SideStep_FingerName_Index(execute=@n24.else, SideStep_FingerName_Index=1) -> (Output_Get=0)
   n30 = Branch(execute=@n23.else, Condition="integer <= integer"(A=Get SideStep_FingerName_Index() -> (SideStep_FingerName_Index=0), B=1))
   > n31 = Branch(Condition=Get SideStep_isMetaCtrl())
   > n32 = Set SideStep_FingerName_Index(SideStep_FingerName_Index=5) -> (Output_Get=0)
   n33 = Branch(execute=[@n30.else, @n31.else], Condition="integer <= integer"(A=Get SideStep_FingerName_Index() -> (SideStep_FingerName_Index=0)))
   > n34 = Set SideStep_FingerName_Index(SideStep_FingerName_Index=5) -> (Output_Get=0)
   n35 = Set SideStep_Direction(execute=@n20.else, SideStep_Direction=Get Direction() -> (Direction=0)) -> (Output_Get=0)
   Branch(execute=@n19.LoopBody, Condition="Equal Exactly (String)"(A=@n19, B=Get SideStep_NamePrefix()))
   > Set SideStep_FingerName_Index(SideStep_FingerName_Index=@n19.Array Index) -> (Output_Get=0)
   Add(execute=@n4.LoopBody, TargetArray:Array of Strings=Get SideStep_FingerNames(), NewItem:"String (by ref)"="To String (Name)"(InName=@n4))
   Set SideStep_CR(execute=@n3.LoopBody, SideStep_CR:Control Rig Object Reference=@n3) -> (Output_Get:Control Rig Object Reference)
   n36 = "To String (Name)"(InName=@n3.Array Element_ControlName_5_#2)
   n37 = Split(SourceString=@n36, InStr=/, SearchDir=FromStart)
   > n38 = Branch(Condition=OR Boolean(A="Equal, Case Insensitive (String)"(A=@n37, B=Body), B="Equal Exactly (String)"(A=@n36, B=Spine/Start IK)))
   > n39 = Branch(Condition="Equal (Integer)"(A=Get Direction() -> (Direction=0), B=1))
   > n40 = IKFK_Check(control to check=Leg L/IK, controlRigToCheck=Get SideStep_CR(), SelectControl=false) -> (TargetControl=None, DestinationControl=None)
   > n41 = Branch(Condition=@n40)
   > Add Unique!(
    execute
    TargetArray:Array of STR Control Rig Name Pair Structures=Get STR_Selection_Output()
    NewItem_ControlRig_2_#1:Control Rig Object Reference=Get SideStep_CR()
    NewItem_ControlName_5_#2:Name=Leg L/IK
   ) -> (then, ReturnValue)
   Add Unique(
    execute=@n41.else
    TargetArray=Get STR_Selection_Output()
    NewItem_ControlRig_2_#1=Get SideStep_CR()
    NewItem_ControlName_5_#2=Leg L/FK 0
   )
   n42 = IKFK_Check(execute=@n39.else, control to check=Leg R/IK, controlRigToCheck=Get SideStep_CR(), SelectControl=false) -> (TargetControl=None, DestinationControl=None)
   > n43 = Branch(Condition=@n42)
   > Add Unique(
    TargetArray=Get STR_Selection_Output()
    NewItem_ControlRig_2_#1=Get SideStep_CR()
    NewItem_ControlName_5_#2=Leg R/IK
   )
   Add Unique(
    execute=@n43.else
    TargetArray=Get STR_Selection_Output()
    NewItem_ControlRig_2_#1=Get SideStep_CR()
    NewItem_ControlName_5_#2=Leg R/FK 0
   )
   n44 = Branch(
    execute=@n38.else
    Condition=OR Boolean(A="Equal, Case Insensitive (String)"(A=@n37, B=Spine), B="Equal, Case Insensitive (String)"(A=@n37, B=Neck))
   )
   > n45 = Branch(Condition="Equal (Integer)"(A=Get Direction() -> (Direction=0), B=1))
   > Add Unique(
    TargetArray=Get STR_Selection_Output()
    NewItem_ControlRig_2_#1=Get SideStep_CR()
    NewItem_ControlName_5_#2=Clavicle L/FK
   )
   Add Unique(
    execute=@n45.else
    TargetArray=Get STR_Selection_Output()
    NewItem_ControlRig_2_#1=Get SideStep_CR()
    NewItem_ControlName_5_#2=Clavicle R/FK
   )
   Add Unique(
    execute=@n44.else
    TargetArray=Get SideStep_Remainder()
    NewItem_ControlRig_2_#1=Get SideStep_CR()
    NewItem_ControlName_5_#2=@n3.Array Element_ControlName_5_#2
   )
   Remove Index!(execute, TargetArray:Array of STR Control Rig Name Pair Structures=Get SideStep_Remainder(), IndexToRemove:int=0) -> (then)
   "integer > integer"(A=Length(TargetArray=Get SideStep_Remainder()))
  ]}
 SpecialShit {nodes [
   SpecialShit() {is=title K2Node_FunctionEntry}
   > n1 = GetSelection(ClearSelection?=false)
   > n2 = For Each Loop!(Exec, Array:Array of STR Control Rig Name Pair Structures=@n1) -> (
    LoopBody
    Array Element_ControlRig_2_#1:Control Rig Object Reference
    Array Element_ControlName_5_#2:Name
    Array Index
    Completed
   )
   > Set Special_CR(Special_CR:Control Rig Object Reference=@n2) -> (Output_Get:Control Rig Object Reference)
   > Set Special_ControlName(Special_ControlName:Name=@n2.Array Element_ControlName_5_#2) -> (Output_Get:Name=None)
   n3 = Get Special_ControlName() -> (Special_ControlName:Name=None)
   n4 = Split(SourceString=Split(SourceString="To String (Name)"(InName=@n3), InStr=/, SearchDir=FromStart), InStr=" ", SearchDir=FromStart)
   > n5 = Branch(Condition=Contains Item(TargetArray=Get FingerNames(), ItemToFind=String To Name(InString=@n4)))
   n6 = "To String (Name)"(InName=Get Special_ControlName() -> (Special_ControlName=None))
   > Set Special_LorR(Special_LorR:String=Split(SourceString=Split(SourceString=@n6, InStr=/, SearchDir=FromStart), InStr=" ", SearchDir=FromStart).RightS) -> (Output_Get:String)
   > n7 = For Each Loop!(Exec, Array:Array of Names=Get Special_Finger_SpreadInner() -> (Special_Finger_SpreadInner:Array of Names)) -> (LoopBody, Array Element:Name, Array Index, Completed)
   n8 = Branch(
    execute=@n7.Completed
    Condition=Contains Item(TargetArray=Get Special_TEMP() -> (Special_TEMP:Array of Names), ItemToFind=Get Special_ControlName() -> (Special_ControlName=None))
   )
   > ClearControlSelection(self=Get Special_CR() -> (Special_CR:Control Rig Object Reference), bSetupUndo=false)
   n9 = Get Special_LorR() -> (Special_LorR:String)
   > SelectControl(
    self=Get Special_CR()
    InControlName=String To Name(InString=Replace(SourceString=Finger Spread Inner Q/Spread, From=Q, To=@n9))
    bSelect=true
    bSetupUndo=false
   )
   Clear(execute=@n8.else, TargetArray:Array of Names=Get Special_TEMP())
   > n10 = For Each Loop(Array=Get Special_Finger_SpreadOuter() -> (Special_Finger_SpreadOuter:Array of Names))
   n11 = Branch(
    execute=@n10.Completed
    Condition=Contains Item(TargetArray=Get Special_TEMP(), ItemToFind=Get Special_ControlName() -> (Special_ControlName=None))
   )
   > ClearControlSelection(self=Get Special_CR(), bSetupUndo=false)
   n12 = Get Special_LorR()
   > SelectControl(
    self=Get Special_CR()
    InControlName=String To Name(InString=Replace(SourceString=Finger Spread Outer Q/Spread, From=Q, To=@n12))
    bSelect=true
    bSetupUndo=false
   )
   Clear(execute=@n11.else, TargetArray=Get Special_TEMP())
   > n13 = For Each Loop(Array=Get Special_Meta_SpreadInner() -> (Special_Meta_SpreadInner:Array of Names))
   n14 = Branch(
    execute=@n13.Completed
    Condition=Contains Item(TargetArray=Get Special_TEMP(), ItemToFind=Get Special_ControlName() -> (Special_ControlName=None))
   )
   > ClearControlSelection(self=Get Special_CR(), bSetupUndo=false)
   n15 = Get Special_LorR()
   > SelectControl(
    self=Get Special_CR()
    InControlName=String To Name(InString=Replace(SourceString=Meta Spread Inner Q/Spread, From=Q, To=@n15))
    bSelect=true
    bSetupUndo=false
   )
   Clear(execute=@n14.else, TargetArray=Get Special_TEMP())
   > n16 = For Each Loop(Array=Get Special_Meta_SpreadOuter() -> (Special_Meta_SpreadOuter:Array of Names))
   Branch(
    execute=@n16.Completed
    Condition=Contains Item(TargetArray=Get Special_TEMP(), ItemToFind=Get Special_ControlName() -> (Special_ControlName=None))
   )
   > ClearControlSelection(self=Get Special_CR(), bSetupUndo=false)
   n17 = Get Special_LorR()
   > SelectControl(
    self=Get Special_CR()
    InControlName=String To Name(InString=Replace(SourceString=Meta Spread Outer Q/Spread, From=Q, To=@n17))
    bSelect=true
    bSetupUndo=false
   )
   n18 = "To String (Name)"(InName=@n16)
   n19 = Get Special_LorR()
   Add Unique!(
    execute=@n16.LoopBody
    TargetArray:Array of Names=Get Special_TEMP()
    NewItem:"Name (by ref)"=String To Name(InString=Replace(SourceString=@n18, From=Q, To=@n19))
   ) -> (then, ReturnValue)
   n20 = "To String (Name)"(InName=@n13)
   n21 = Get Special_LorR()
   Add Unique(execute=@n13.LoopBody, TargetArray=Get Special_TEMP(), NewItem=String To Name(InString=Replace(SourceString=@n20, From=Q, To=@n21)))
   n22 = "To String (Name)"(InName=@n10)
   n23 = Get Special_LorR()
   Add Unique(execute=@n10.LoopBody, TargetArray=Get Special_TEMP(), NewItem=String To Name(InString=Replace(SourceString=@n22, From=Q, To=@n23)))
   n24 = "To String (Name)"(InName=@n7)
   n25 = Get Special_LorR()
   Add Unique(execute=@n7.LoopBody, TargetArray=Get Special_TEMP(), NewItem=String To Name(InString=Replace(SourceString=@n24, From=Q, To=@n25)))
   n26 = Branch(execute=@n5.else, Condition="Equal (Name)"(A=Get Special_ControlName() -> (Special_ControlName=None), B=Body/Body))
   > ClearControlSelection(self=Get Special_CR(), bSetupUndo=false)
   > SelectControl(self=Get Special_CR(), InControlName=String To Name(InString=Prop/Prop Global), bSelect=true, bSetupUndo=false)
   ClearControlSelection(execute=@n26.else, self=Get Special_CR(), bSetupUndo=false)
   > SelectControl(self=Get Special_CR(), InControlName=String To Name(InString=Body/Body), bSelect=true, bSetupUndo=false)
  ]}
]
variables [
 Arm_L_Attach {type="array<string>", default [hand_attach_l_ctrl, hand_attach_l_ctrl, hand_attach_l_parent_ctrl, hand_l_ctrl]}
 Arm_L_FK {type="array<string>", default [Attach/Hand Attach L, Arm L/FK 2, Arm L/FK 1, Arm L/FK 0, Clavicle L/FK, Spine/Fk 3]}
 Arm_L_IK {type="array<string>", default [hand_attach_l_parent_ctrl, hand_ik_l_ctrl, arm_pv_l_ctrl, arm_shrug_ik_l_ctrl, torso_ctrl]}
 Arm_R_Attach {type="array<string>", default [hand_attach_r_ctrl, hand_attach_r_ctrl, hand_attach_r_parent_ctrl, hand_r_ctrl]}
 Arm_R_FK {type="array<string>", default [hand_attach_r_parent_ctrl, hand_r_ctrl, arm_r_ctrl, arm_shrug_r_ctrl, torso_ctrl]}
 Arm_R_IK {type="array<string>", default [hand_attach_r_parent_ctrl, hand_ik_r_ctrl, arm_pv_r_ctrl, arm_shrug_ik_r_ctrl, torso_ctrl]}
 Control Rig {type=ControlRig, default=null}
 ControlRigs_Input {type="array<ControlRig>", default []}
 Core {type="array<string>", default [root_ctrl,root_ctrl,global_ctrl,offset_ctrl,body_ctrl,torso_ctrl,head_ctrl,mouth_ctrl]}
 Current CR Control List {type="array<string>", default []}
 CurrLevelSequence {type=LevelSequence, default=null}
 Debug_Check {type=boolean}
 Direction {type=integer, default=1}
 Face {
  type="array<string>"
  default [
   None
   mouth_ctrl
   eye_mask_r_ctrl
   eye_mask_l_ctrl
   eye_l_ctrl
   eye_r_ctrl
   eyelash_l_ctrl
   eyelash_r_ctrl
   brow_l_ctrl
   brow_r_ctrl
   accent_l_lower_ctrl
   accent_r_lower_ctrl
   accent_l_upper_ctrl
   accent_r_upper_ctrl
   None
  ]
 }
 FingerControls {type="array<string>", default [Pinky,Index,Middle,Ring,Pinky]}
 FingerNames {type="array<string>", default [Meta,Thumb,Index,Middle,Ring,Pinky]}
 Hierarchy {type=RigHierarchy, default=null}
 IKCheckList {type="array<string>", default []}
 Leg_L_FK {type="array<string>", default [leg_r_ctrl,leg_l_ctrl,body_ctrl]}
 Leg_L_IK {type="array<string>", default [leg_r_ctrl, leg_ik_l_ctrl, body_ctrl]}
 Leg_R_FK {type="array<string>", default [leg_l_ctrl,leg_r_ctrl,body_ctrl]}
 Leg_R_IK {type="array<string>", default [leg_l_ctrl, leg_ik_r_ctrl, body_ctrl]}
 NewVar {type=integer}
 NonChain_List {type="array<string>", default []}
 selectedControl_parsed {type=string, default=None}
 Selection_Input {type="array<string>", default []}
 Selection_New {type="array<string>", default []}
 SET_Map {
  type="array<STR_PickWalk_SET>"
  default [
   {nextControl=Attach/Hand Attach L, previousControl=Spine/End FK, controlSET [Arm L/IK, Arm L/PV, Clavicle L/FK]}
   {nextControl=Attach/Hand Attach L, previousControl=Spine/End FK, controlSET [Arm L/FK 2, Arm L/FK 1, Arm L/FK 0, Clavicle L/FK]}
   {nextControl=Index L/FK 0, previousControl=Arm L/FK 2, controlSET [Attach/Hand Attach L]}
   {nextControl=Attach/Hand Attach R, previousControl=Spine/End FK, controlSET [Arm R/IK, Arm R/PV, Clavicle R/FK]}
   {nextControl=Attach/Hand Attach R, previousControl=Spine/End FK, controlSET [Arm R/FK 2, Arm R/FK 1, Arm R/FK 0, Clavicle R/FK]}
   {nextControl=Index R/FK 0, previousControl=Arm R/FK 2, controlSET [Attach/Hand Attach R]}
   {nextControl=Body/Body, previousControl=Neck/Start FK, controlSET [Spine/Start FK, Spine/Mid FK, Spine/End FK]}
   {nextControl=Spine/End FK, previousControl=Neck/End FK, controlSET [Neck/Start FK, Neck/Mid FK, Neck/End FK]}
   {nextControl=root/Local, previousControl=Spine/Mid FK, controlSET [Spine/Start IK, Body/Body]}
   {nextControl=root/Global, previousControl=Body/Body, controlSET [root/Global, root/Root, root/Local]}
   {nextControl=Foot L/Ball FK, previousControl=Body/Body, controlSET [Foot L/Ball FK, Leg L/FK 2, Leg L/FK 1, Leg L/FK 0]}
   {nextControl=Foot L/Foot Rocker, previousControl=Body/Body, controlSET [Leg L/IK, Leg L/PV]}
   {nextControl=Foot L/Ball FK, previousControl=Leg L/IK, controlSET [Foot L/Ball FK, Foot L/Ball IK, Foot L/Foot Rocker]}
   {nextControl=Foot R/Ball FK, previousControl=Body/Body, controlSET [Foot R/Ball FK, Leg R/FK 2, Leg R/FK 1, Leg R/FK 0]}
   {nextControl=Foot R/Foot Rocker, previousControl=Body/Body, controlSET [Leg R/IK, Leg R/PV]}
   {nextControl=Foot R/Ball FK, previousControl=Leg R/IK, controlSET [Foot R/Ball FK, Foot R/Ball IK, Foot R/Foot Rocker]}
   {nextControl=Body/Body, previousControl=Neck/Start FK, controlSET [Spine/Mid IK, Spine/End IK]}
   {nextControl=Prop/Aim, previousControl=Prop/Prop, controlSET [Prop/Prop, Prop/Prop Local, Prop/Prop Global, Prop/Aim]}
  ]
 }
 SET_Map_Hand {
  type="array<STR_PickWalk_SET>"
  default [
   {nextControl=Index Q/FK 2, previousControl=Arm Q/FK 2, controlSET [Index Q/FK 2, Index Q/FK 1, Index Q/FK 0, Meta Q/Index Metacarpal Q]}
   {nextControl=Index Q/FK 2, previousControl=Arm Q/FK 2, controlSET [Middle Q/FK 2, Middle Q/FK 1, Middle Q/FK 0, Meta Q/Middle Metacarpal Q]}
   {nextControl=Ring Q/FK 2, previousControl=Arm Q/FK 2, controlSET [Ring Q/FK 2, Ring Q/FK 1, Ring Q/FK 0, Meta Q/Ring Metacarpal Q]}
   {nextControl=Pinky Q/FK 2, previousControl=Arm Q/FK 2, controlSET [Pinky Q/FK 2, Pinky Q/FK 1, Pinky Q/FK 0, Meta Q/Pinky Metacarpal Q]}
   {nextControl=Index Q/FK 2, previousControl=Arm Q/FK 2, controlSET [Thumb Q/FK 2, Thumb Q/FK 1, Thumb Q/FK 0]}
  ]
 }
 SET_Map_IKIFK {
  type="array<STR_PickwalkMap_IKFK>"
  default [
   {destinationControl=Arm L/PV, targetControl=Arm L/FK 0}
   {destinationControl=Arm L/PV, targetControl=Arm L/FK 1}
   {destinationControl=Arm L/IK, targetControl=Arm L/FK 2}
   {destinationControl=Arm R/PV, targetControl=Arm R/FK 0}
   {destinationControl=Arm R/PV, targetControl=Arm R/FK 1}
   {destinationControl=Arm R/IK, targetControl=Arm R/FK 2}
   {destinationControl=Leg L/IK, targetControl=Leg L/FK 0}
   {destinationControl=Leg L/IK, targetControl=Leg L/FK 2}
   {destinationControl=Leg R/IK, targetControl=Leg R/FK 0}
   {destinationControl=Leg R/IK, targetControl=Leg R/FK 2}
   {destinationControl=Spine/End IK, targetControl=Spine/End FK}
   {destinationControl=Spine/Mid IK, targetControl=Spine/Start FK}
   {destinationControl=Spine/Mid IK, targetControl=Spine/Mid FK}
  ]
 }
 STR_Chainwalk {type="array<STR_ControlRigNamePair>", default []}
 STR_Hardwalk {type="array<STR_ControlRigNamePair>", default []}
 STR_Selection_Input {type="array<STR_ControlRigNamePair>", default []}
 STR_Selection_Output {type="array<STR_ControlRigNamePair>", default []}
 World Context Object {type=Object, default=null}
]
widgets {
 info {rootWidgetClass=&Border, widgetCount=24}
 widgets [
  Border_0 {
   widgetClassPath=&Border
   widget=Border_0
   properties {
    horizontalAlignment=HAlign_Center
    verticalAlignment=VAlign_Center
    background {
     bIsDynamicallyLoaded=false
     drawAs=Image
     imageType=NoImage
     mirroring=NoMirror
     resourceName=None
     resourceObject=null
     tiling=NoTile
     imageSize=<32 32>
     margin=<ltrb 0 0 0 0>
     outlineSettings {
      bUseBrushTransparency=false
      roundingType=HalfHeightRadius
      width=0
      color {colorUseRule=UseColor_Specified, specifiedColor=<rgba 0 0 0 0>}
      cornerRadii=<0 0 0 0>
     }
     tintColor {colorUseRule=UseColor_Specified, specifiedColor=<rgba 0.0176 0.0176 0.0176 1>}
     uVRegion {bIsValid=false, max=<0 0>, min=<0 0>}
    }
    padding=<ltrb 10 10 10 10>
   }
  }
  ScaleBox_122 {
   widgetClassPath=&ScaleBox
   parent=Border_0
   widget=ScaleBox_122
   slot {class=BorderSlot, properties {horizontalAlignment=HAlign_Center, verticalAlignment=VAlign_Center, padding=<ltrb 10 10 10 10>}}
  }
  HorizontalBox_126 {
   widgetClassPath=&HorizontalBox
   parent=ScaleBox_122
   widget=HorizontalBox_126
   slot {class=ScaleBoxSlot, properties {horizontalAlignment=HAlign_Fill, verticalAlignment=VAlign_Fill}}
  }
  VerticalBox_102 {widgetClassPath=&VerticalBox, parent=HorizontalBox_126, widget=VerticalBox_102, slot {class=HorizontalBoxSlot}}
  BTN_Mirror {
   widgetClassPath=&EditorUtilityButton
   parent=VerticalBox_102
   widget=BTN_Mirror
   properties {
    toolTipText=Q
    backgroundColor=<rgba 1 1 1 0.2>
    colorAndOpacity=<rgba 0.0648 0.2502 0.2346 0.7>
    widgetStyle {
     clickedSlateSound {resourceObject=null}
     disabled {
      bIsDynamicallyLoaded=false
      drawAs=RoundedBox
      imageType=FullColor
      mirroring=NoMirror
      resourceName=None
      resourceObject=null
      tiling=NoTile
      imageSize=<64 64>
      margin=<ltrb 0 0 0 0>
      outlineSettings {
       bUseBrushTransparency=false
       roundingType=FixedRadius
       width=3
       color {colorUseRule=UseColor_Specified, specifiedColor=<rgba 0 0 0 0.2>}
       cornerRadii=<20 20 20 20>
      }
      tintColor {colorUseRule=UseColor_Specified, specifiedColor=<rgba 0.0395 0.0395 0.0395 1>}
      uVRegion {bIsValid=false, max=<0 0>, min=<0 0>}
     }
     disabledForeground {colorUseRule=UseColor_ColorTable, specifiedColor=<rgba 0 0 0 0>}
     hovered {
      bIsDynamicallyLoaded=false
      drawAs=RoundedBox
      imageType=FullColor
      mirroring=NoMirror
      resourceName=None
      resourceObject=null
      tiling=NoTile
      imageSize=<64 64>
      margin=<ltrb 0 0 0 0>
      outlineSettings {
       bUseBrushTransparency=false
       roundingType=FixedRadius
       width=3
       color {colorUseRule=UseColor_Specified, specifiedColor=<rgba 0.7222 0.7222 0.7222 0.2>}
       cornerRadii=<20 20 20 20>
      }
      tintColor {colorUseRule=UseColor_Specified, specifiedColor=<rgba 0.0953 0.0953 0.0953 1>}
      uVRegion {bIsValid=false, max=<0 0>, min=<0 0>}
     }
     hoveredForeground {colorUseRule=UseColor_ColorTable, specifiedColor=<rgba 0 0 0 0>}
     hoveredSlateSound {resourceObject=null}
     normal {
      bIsDynamicallyLoaded=false
      drawAs=RoundedBox
      imageType=FullColor
      mirroring=NoMirror
      resourceName=None
      resourceObject=null
      tiling=NoTile
      imageSize=<64 64>
      margin=<ltrb 0 0 0 0>
      outlineSettings {
       bUseBrushTransparency=false
       roundingType=FixedRadius
       width=15
       color {colorUseRule=UseColor_Specified, specifiedColor=<rgba 0.7222 0.7222 0.7222 0.1>}
       cornerRadii=<30 30 30 30>
      }
      tintColor {colorUseRule=UseColor_Specified, specifiedColor=<rgba 0.0395 0.0395 0.0395 1>}
      uVRegion {bIsValid=false, max=<0 0>, min=<0 0>}
     }
     normalForeground {colorUseRule=UseColor_ColorTable, specifiedColor=<rgba 0 0 0 0>}
     normalPadding=<ltrb 12 1.5 12 1.5>
     pressed {
      bIsDynamicallyLoaded=false
      drawAs=RoundedBox
      imageType=FullColor
      mirroring=NoMirror
      resourceName=None
      resourceObject=null
      tiling=NoTile
      imageSize=<64 64>
      margin=<ltrb 0 0 0 0>
      outlineSettings {
       bUseBrushTransparency=false
       roundingType=FixedRadius
       width=3
       color {colorUseRule=UseColor_Specified, specifiedColor=<rgba 0.7222 0.7222 0.7222 0.2>}
       cornerRadii=<20 20 20 20>
      }
      tintColor {colorUseRule=UseColor_Specified, specifiedColor=<rgba 0.0284 0.0284 0.0284 1>}
      uVRegion {bIsValid=false, max=<0 0>, min=<0 0>}
     }
     pressedForeground {colorUseRule=UseColor_ColorTable, specifiedColor=<rgba 0 0 0 0>}
     pressedPadding=<ltrb 12 2.5 12 0.5>
     pressedSlateSound {resourceObject=null}
    }
   }
   slot {class=VerticalBoxSlot, properties {padding=<ltrb 20 20 20 20>}}
  }
  Image_4 {
   widgetClassPath=&Image
   bIsVariable=true
   parent=BTN_Mirror
   widget=Image_4
   properties {brush {
     bIsDynamicallyLoaded=false
     drawAs=Image
     imageType=FullColor
     mirroring=NoMirror
     resourceName=None
     resourceObject=/AnimInEdTools/Pickwalk_ControlRig/icons/AnimToolbox_UI_Pickwalk_Mirror
     tiling=NoTile
     imageSize=<600 600>
     margin=<ltrb 0 0 0 0>
     outlineSettings {
      bUseBrushTransparency=false
      roundingType=HalfHeightRadius
      width=0
      color {colorUseRule=UseColor_Specified, specifiedColor=<rgba 0 0 0 0>}
      cornerRadii=<0 0 0 0>
     }
     tintColor {colorUseRule=UseColor_Specified, specifiedColor=<rgba 1 1 1 1>}
     uVRegion {bIsValid=false, max=<0 0>, min=<0 0>}
    }}
   slot {class=ButtonSlot, properties {horizontalAlignment=HAlign_Fill, verticalAlignment=VAlign_Fill}}
  }
  BTN_PW_Left {
   widgetClassPath=&EditorUtilityButton
   parent=VerticalBox_102
   widget=BTN_PW_Left
   properties {
    toolTipText=A
    backgroundColor=<rgba 1 1 1 0.2>
    colorAndOpacity=<rgba 0.0865 0.1845 0.3515 0.7>
    widgetStyle {
     clickedSlateSound {resourceObject=null}
     disabled {
      bIsDynamicallyLoaded=false
      drawAs=RoundedBox
      imageType=FullColor
      mirroring=NoMirror
      resourceName=None
      resourceObject=null
      tiling=NoTile
      imageSize=<64 64>
      margin=<ltrb 0 0 0 0>
      outlineSettings {
       bUseBrushTransparency=false
       roundingType=FixedRadius
       width=3
       color {colorUseRule=UseColor_Specified, specifiedColor=<rgba 0 0 0 0.2>}
       cornerRadii=<20 20 20 20>
      }
      tintColor {colorUseRule=UseColor_Specified, specifiedColor=<rgba 0.0395 0.0395 0.0395 1>}
      uVRegion {bIsValid=false, max=<0 0>, min=<0 0>}
     }
     disabledForeground {colorUseRule=UseColor_ColorTable, specifiedColor=<rgba 0 0 0 0>}
     hovered {
      bIsDynamicallyLoaded=false
      drawAs=RoundedBox
      imageType=FullColor
      mirroring=NoMirror
      resourceName=None
      resourceObject=null
      tiling=NoTile
      imageSize=<64 64>
      margin=<ltrb 0 0 0 0>
      outlineSettings {
       bUseBrushTransparency=false
       roundingType=FixedRadius
       width=3
       color {colorUseRule=UseColor_Specified, specifiedColor=<rgba 0.7222 0.7222 0.7222 0.2>}
       cornerRadii=<20 20 20 20>
      }
      tintColor {colorUseRule=UseColor_Specified, specifiedColor=<rgba 0.0953 0.0953 0.0953 1>}
      uVRegion {bIsValid=false, max=<0 0>, min=<0 0>}
     }
     hoveredForeground {colorUseRule=UseColor_ColorTable, specifiedColor=<rgba 0 0 0 0>}
     hoveredSlateSound {resourceObject=null}
     normal {
      bIsDynamicallyLoaded=false
      drawAs=RoundedBox
      imageType=FullColor
      mirroring=NoMirror
      resourceName=None
      resourceObject=null
      tiling=NoTile
      imageSize=<64 64>
      margin=<ltrb 0 0 0 0>
      outlineSettings {
       bUseBrushTransparency=false
       roundingType=FixedRadius
       width=15
       color {colorUseRule=UseColor_Specified, specifiedColor=<rgba 0.7222 0.7222 0.7222 0.1>}
       cornerRadii=<30 30 30 30>
      }
      tintColor {colorUseRule=UseColor_Specified, specifiedColor=<rgba 0.0395 0.0395 0.0395 1>}
      uVRegion {bIsValid=false, max=<0 0>, min=<0 0>}
     }
     normalForeground {colorUseRule=UseColor_ColorTable, specifiedColor=<rgba 0 0 0 0>}
     normalPadding=<ltrb 12 1.5 12 1.5>
     pressed {
      bIsDynamicallyLoaded=false
      drawAs=RoundedBox
      imageType=FullColor
      mirroring=NoMirror
      resourceName=None
      resourceObject=null
      tiling=NoTile
      imageSize=<64 64>
      margin=<ltrb 0 0 0 0>
      outlineSettings {
       bUseBrushTransparency=false
       roundingType=FixedRadius
       width=3
       color {colorUseRule=UseColor_Specified, specifiedColor=<rgba 0.7222 0.7222 0.7222 0.2>}
       cornerRadii=<20 20 20 20>
      }
      tintColor {colorUseRule=UseColor_Specified, specifiedColor=<rgba 0.0284 0.0284 0.0284 1>}
      uVRegion {bIsValid=false, max=<0 0>, min=<0 0>}
     }
     pressedForeground {colorUseRule=UseColor_ColorTable, specifiedColor=<rgba 0 0 0 0>}
     pressedPadding=<ltrb 12 2.5 12 0.5>
     pressedSlateSound {resourceObject=null}
    }
   }
   slot {class=VerticalBoxSlot, properties {padding=<ltrb 20 20 20 20>}}
  }
  Image_5 {
   widgetClassPath=&Image
   bIsVariable=true
   parent=BTN_PW_Left
   widget=Image_5
   properties {brush {
     bIsDynamicallyLoaded=false
     drawAs=Image
     imageType=FullColor
     mirroring=NoMirror
     resourceName=None
     resourceObject=/AnimInEdTools/Pickwalk_ControlRig/icons/AnimToolbox_UI_Pickwalk_Left
     tiling=NoTile
     imageSize=<600 600>
     margin=<ltrb 0 0 0 0>
     outlineSettings {
      bUseBrushTransparency=false
      roundingType=HalfHeightRadius
      width=0
      color {colorUseRule=UseColor_Specified, specifiedColor=<rgba 0 0 0 0>}
      cornerRadii=<0 0 0 0>
     }
     tintColor {colorUseRule=UseColor_Specified, specifiedColor=<rgba 1 1 1 1>}
     uVRegion {bIsValid=false, max=<0 0>, min=<0 0>}
    }}
   slot {class=ButtonSlot, properties {horizontalAlignment=HAlign_Fill, verticalAlignment=VAlign_Fill}}
  }
  Test_Bl {
   widgetClassPath=&EditorUtilityButton
   parent=VerticalBox_102
   widget=Test_Bl
   properties {
    toolTipText=Home
    backgroundColor=<rgba 1 1 1 0.2>
    colorAndOpacity=<rgba 0.3278 0.0865 0.1144 0.7>
    widgetStyle {
     clickedSlateSound {resourceObject=null}
     disabled {
      bIsDynamicallyLoaded=false
      drawAs=RoundedBox
      imageType=FullColor
      mirroring=NoMirror
      resourceName=None
      resourceObject=null
      tiling=NoTile
      imageSize=<64 64>
      margin=<ltrb 0 0 0 0>
      outlineSettings {
       bUseBrushTransparency=false
       roundingType=FixedRadius
       width=3
       color {colorUseRule=UseColor_Specified, specifiedColor=<rgba 0 0 0 0.2>}
       cornerRadii=<20 20 20 20>
      }
      tintColor {colorUseRule=UseColor_Specified, specifiedColor=<rgba 0.0395 0.0395 0.0395 1>}
      uVRegion {bIsValid=false, max=<0 0>, min=<0 0>}
     }
     disabledForeground {colorUseRule=UseColor_ColorTable, specifiedColor=<rgba 0 0 0 0>}
     hovered {
      bIsDynamicallyLoaded=false
      drawAs=RoundedBox
      imageType=FullColor
      mirroring=NoMirror
      resourceName=None
      resourceObject=null
      tiling=NoTile
      imageSize=<64 64>
      margin=<ltrb 0 0 0 0>
      outlineSettings {
       bUseBrushTransparency=false
       roundingType=FixedRadius
       width=3
       color {colorUseRule=UseColor_Specified, specifiedColor=<rgba 0.7222 0.7222 0.7222 0.2>}
       cornerRadii=<20 20 20 20>
      }
      tintColor {colorUseRule=UseColor_Specified, specifiedColor=<rgba 0.0953 0.0953 0.0953 1>}
      uVRegion {bIsValid=false, max=<0 0>, min=<0 0>}
     }
     hoveredForeground {colorUseRule=UseColor_ColorTable, specifiedColor=<rgba 0 0 0 0>}
     hoveredSlateSound {resourceObject=null}
     normal {
      bIsDynamicallyLoaded=false
      drawAs=RoundedBox
      imageType=FullColor
      mirroring=NoMirror
      resourceName=None
      resourceObject=null
      tiling=NoTile
      imageSize=<64 64>
      margin=<ltrb 0 0 0 0>
      outlineSettings {
       bUseBrushTransparency=false
       roundingType=FixedRadius
       width=15
       color {colorUseRule=UseColor_Specified, specifiedColor=<rgba 0.7222 0.7222 0.7222 0.1>}
       cornerRadii=<30 30 30 30>
      }
      tintColor {colorUseRule=UseColor_Specified, specifiedColor=<rgba 0.0395 0.0395 0.0395 1>}
      uVRegion {bIsValid=false, max=<0 0>, min=<0 0>}
     }
     normalForeground {colorUseRule=UseColor_ColorTable, specifiedColor=<rgba 0 0 0 0>}
     normalPadding=<ltrb 12 1.5 12 1.5>
     pressed {
      bIsDynamicallyLoaded=false
      drawAs=RoundedBox
      imageType=FullColor
      mirroring=NoMirror
      resourceName=None
      resourceObject=null
      tiling=NoTile
      imageSize=<64 64>
      margin=<ltrb 0 0 0 0>
      outlineSettings {
       bUseBrushTransparency=false
       roundingType=FixedRadius
       width=3
       color {colorUseRule=UseColor_Specified, specifiedColor=<rgba 0.7222 0.7222 0.7222 0.2>}
       cornerRadii=<20 20 20 20>
      }
      tintColor {colorUseRule=UseColor_Specified, specifiedColor=<rgba 0.0284 0.0284 0.0284 1>}
      uVRegion {bIsValid=false, max=<0 0>, min=<0 0>}
     }
     pressedForeground {colorUseRule=UseColor_ColorTable, specifiedColor=<rgba 0 0 0 0>}
     pressedPadding=<ltrb 12 2.5 12 0.5>
     pressedSlateSound {resourceObject=null}
    }
   }
   slot {class=VerticalBoxSlot, properties {padding=<ltrb 20 20 20 20>}}
  }
  Image_6 {
   widgetClassPath=&Image
   bIsVariable=true
   parent=Test_Bl
   widget=Image_6
   properties {brush {
     bIsDynamicallyLoaded=false
     drawAs=Image
     imageType=FullColor
     mirroring=NoMirror
     resourceName=None
     resourceObject=/AnimInEdTools/Pickwalk_ControlRig/icons/AnimToolbox_UI_Pickwalk_Debug
     tiling=NoTile
     imageSize=<600 600>
     margin=<ltrb 0 0 0 0>
     outlineSettings {
      bUseBrushTransparency=false
      roundingType=HalfHeightRadius
      width=0
      color {colorUseRule=UseColor_Specified, specifiedColor=<rgba 0 0 0 0>}
      cornerRadii=<0 0 0 0>
     }
     tintColor {colorUseRule=UseColor_Specified, specifiedColor=<rgba 1 1 1 1>}
     uVRegion {bIsValid=false, max=<0 0>, min=<0 0>}
    }}
   slot {class=ButtonSlot, properties {horizontalAlignment=HAlign_Fill, verticalAlignment=VAlign_Fill}}
  }
  VerticalBox {widgetClassPath=&VerticalBox, parent=HorizontalBox_126, widget=VerticalBox, slot {class=HorizontalBoxSlot}}
  BTN_PW_Up {
   widgetClassPath=&EditorUtilityButton
   parent=VerticalBox
   widget=BTN_PW_Up
   properties {
    toolTipText=W
    backgroundColor=<rgba 1 1 1 0.2>
    colorAndOpacity=<rgba 0.0865 0.1845 0.3515 0.7>
    widgetStyle {
     clickedSlateSound {resourceObject=null}
     disabled {
      bIsDynamicallyLoaded=false
      drawAs=RoundedBox
      imageType=NoImage
      mirroring=NoMirror
      resourceName=None
      resourceObject=null
      tiling=NoTile
      imageSize=<0 0>
      margin=<ltrb 0 0 0 0>
      outlineSettings {
       bUseBrushTransparency=false
       roundingType=FixedRadius
       width=3
       color {colorUseRule=UseColor_Specified, specifiedColor=<rgba 0 0 0 0.2>}
       cornerRadii=<20 20 20 20>
      }
      tintColor {colorUseRule=UseColor_Specified, specifiedColor=<rgba 0.0395 0.0395 0.0395 1>}
      uVRegion {bIsValid=false, max=<0 0>, min=<0 0>}
     }
     disabledForeground {colorUseRule=UseColor_ColorTable, specifiedColor=<rgba 0 0 0 0>}
     hovered {
      bIsDynamicallyLoaded=false
      drawAs=RoundedBox
      imageType=NoImage
      mirroring=NoMirror
      resourceName=None
      resourceObject=null
      tiling=NoTile
      imageSize=<32 32>
      margin=<ltrb 0 0 0 0>
      outlineSettings {
       bUseBrushTransparency=false
       roundingType=FixedRadius
       width=3
       color {colorUseRule=UseColor_Specified, specifiedColor=<rgba 0.7222 0.7222 0.7222 0.2>}
       cornerRadii=<20 20 20 20>
      }
      tintColor {colorUseRule=UseColor_Specified, specifiedColor=<rgba 0.0953 0.0953 0.0953 1>}
      uVRegion {bIsValid=false, max=<0 0>, min=<0 0>}
     }
     hoveredForeground {colorUseRule=UseColor_ColorTable, specifiedColor=<rgba 0 0 0 0>}
     hoveredSlateSound {resourceObject=null}
     normal {
      bIsDynamicallyLoaded=false
      drawAs=RoundedBox
      imageType=NoImage
      mirroring=NoMirror
      resourceName=None
      resourceObject=null
      tiling=NoTile
      imageSize=<32 32>
      margin=<ltrb 0 0 0 0>
      outlineSettings {
       bUseBrushTransparency=false
       roundingType=FixedRadius
       width=15
       color {colorUseRule=UseColor_Specified, specifiedColor=<rgba 0.7222 0.7222 0.7222 0.1>}
       cornerRadii=<30 30 30 30>
      }
      tintColor {colorUseRule=UseColor_Specified, specifiedColor=<rgba 0.0395 0.0395 0.0395 1>}
      uVRegion {bIsValid=false, max=<0 0>, min=<0 0>}
     }
     normalForeground {colorUseRule=UseColor_ColorTable, specifiedColor=<rgba 0 0 0 0>}
     normalPadding=<ltrb 12 1.5 12 1.5>
     pressed {
      bIsDynamicallyLoaded=false
      drawAs=RoundedBox
      imageType=NoImage
      mirroring=NoMirror
      resourceName=None
      resourceObject=null
      tiling=NoTile
      imageSize=<32 32>
      margin=<ltrb 0 0 0 0>
      outlineSettings {
       bUseBrushTransparency=false
       roundingType=FixedRadius
       width=3
       color {colorUseRule=UseColor_Specified, specifiedColor=<rgba 0.7222 0.7222 0.7222 0.2>}
       cornerRadii=<20 20 20 20>
      }
      tintColor {colorUseRule=UseColor_Specified, specifiedColor=<rgba 0.0284 0.0284 0.0284 1>}
      uVRegion {bIsValid=false, max=<0 0>, min=<0 0>}
     }
     pressedForeground {colorUseRule=UseColor_ColorTable, specifiedColor=<rgba 0 0 0 0>}
     pressedPadding=<ltrb 12 2.5 12 0.5>
     pressedSlateSound {resourceObject=null}
    }
   }
   slot {class=VerticalBoxSlot, properties {padding=<ltrb 20 20 20 20>}}
  }
  Image_160 {
   widgetClassPath=&Image
   bIsVariable=true
   parent=BTN_PW_Up
   widget=Image_160
   properties {brush {
     bIsDynamicallyLoaded=false
     drawAs=Image
     imageType=FullColor
     mirroring=NoMirror
     resourceName=None
     resourceObject=/AnimInEdTools/Pickwalk_ControlRig/icons/AnimToolbox_UI_Pickwalk_Up
     tiling=NoTile
     imageSize=<600 600>
     margin=<ltrb 0 0 0 0>
     outlineSettings {
      bUseBrushTransparency=false
      roundingType=HalfHeightRadius
      width=0
      color {colorUseRule=UseColor_Specified, specifiedColor=<rgba 0 0 0 0>}
      cornerRadii=<0 0 0 0>
     }
     tintColor {colorUseRule=UseColor_Specified, specifiedColor=<rgba 1 1 1 1>}
     uVRegion {bIsValid=false, max=<0 0>, min=<0 0>}
    }}
   slot {class=ButtonSlot, properties {horizontalAlignment=HAlign_Fill, verticalAlignment=VAlign_Fill}}
  }
  BTN_RecallSel {
   widgetClassPath=&EditorUtilityButton
   parent=VerticalBox
   widget=BTN_RecallSel
   properties {
    toolTipText=Z
    backgroundColor=<rgba 1 1 1 0.2>
    colorAndOpacity=<rgba 0.3278 0.1301 0.0685 0.7>
    widgetStyle {
     clickedSlateSound {resourceObject=null}
     disabled {
      bIsDynamicallyLoaded=false
      drawAs=RoundedBox
      imageType=NoImage
      mirroring=NoMirror
      resourceName=None
      resourceObject=null
      tiling=NoTile
      imageSize=<0 0>
      margin=<ltrb 0 0 0 0>
      outlineSettings {
       bUseBrushTransparency=false
       roundingType=FixedRadius
       width=3
       color {colorUseRule=UseColor_Specified, specifiedColor=<rgba 0 0 0 0.2>}
       cornerRadii=<20 20 20 20>
      }
      tintColor {colorUseRule=UseColor_Specified, specifiedColor=<rgba 0.0395 0.0395 0.0395 1>}
      uVRegion {bIsValid=false, max=<0 0>, min=<0 0>}
     }
     disabledForeground {colorUseRule=UseColor_ColorTable, specifiedColor=<rgba 0 0 0 0>}
     hovered {
      bIsDynamicallyLoaded=false
      drawAs=RoundedBox
      imageType=NoImage
      mirroring=NoMirror
      resourceName=None
      resourceObject=null
      tiling=NoTile
      imageSize=<32 32>
      margin=<ltrb 0 0 0 0>
      outlineSettings {
       bUseBrushTransparency=false
       roundingType=FixedRadius
       width=3
       color {colorUseRule=UseColor_Specified, specifiedColor=<rgba 0.7222 0.7222 0.7222 0.2>}
       cornerRadii=<20 20 20 20>
      }
      tintColor {colorUseRule=UseColor_Specified, specifiedColor=<rgba 0.0953 0.0953 0.0953 1>}
      uVRegion {bIsValid=false, max=<0 0>, min=<0 0>}
     }
     hoveredForeground {colorUseRule=UseColor_ColorTable, specifiedColor=<rgba 0 0 0 0>}
     hoveredSlateSound {resourceObject=null}
     normal {
      bIsDynamicallyLoaded=false
      drawAs=RoundedBox
      imageType=NoImage
      mirroring=NoMirror
      resourceName=None
      resourceObject=null
      tiling=NoTile
      imageSize=<32 32>
      margin=<ltrb 0 0 0 0>
      outlineSettings {
       bUseBrushTransparency=false
       roundingType=FixedRadius
       width=15
       color {colorUseRule=UseColor_Specified, specifiedColor=<rgba 0.7222 0.7222 0.7222 0.1>}
       cornerRadii=<30 30 30 30>
      }
      tintColor {colorUseRule=UseColor_Specified, specifiedColor=<rgba 0.0395 0.0395 0.0395 1>}
      uVRegion {bIsValid=false, max=<0 0>, min=<0 0>}
     }
     normalForeground {colorUseRule=UseColor_ColorTable, specifiedColor=<rgba 0 0 0 0>}
     normalPadding=<ltrb 12 1.5 12 1.5>
     pressed {
      bIsDynamicallyLoaded=false
      drawAs=RoundedBox
      imageType=NoImage
      mirroring=NoMirror
      resourceName=None
      resourceObject=null
      tiling=NoTile
      imageSize=<32 32>
      margin=<ltrb 0 0 0 0>
      outlineSettings {
       bUseBrushTransparency=false
       roundingType=FixedRadius
       width=3
       color {colorUseRule=UseColor_Specified, specifiedColor=<rgba 0.7222 0.7222 0.7222 0.2>}
       cornerRadii=<20 20 20 20>
      }
      tintColor {colorUseRule=UseColor_Specified, specifiedColor=<rgba 0.0284 0.0284 0.0284 1>}
      uVRegion {bIsValid=false, max=<0 0>, min=<0 0>}
     }
     pressedForeground {colorUseRule=UseColor_ColorTable, specifiedColor=<rgba 0 0 0 0>}
     pressedPadding=<ltrb 12 2.5 12 0.5>
     pressedSlateSound {resourceObject=null}
    }
   }
   slot {class=VerticalBoxSlot, properties {padding=<ltrb 20 20 20 20>}}
  }
  Image_60 {
   widgetClassPath=&Image
   bIsVariable=true
   parent=BTN_RecallSel
   widget=Image_60
   properties {
    brush {
     bIsDynamicallyLoaded=false
     drawAs=Image
     imageType=FullColor
     mirroring=NoMirror
     resourceName=None
     resourceObject=/AnimInEdTools/Pickwalk_ControlRig/icons/AnimToolbox_UI_Pickwalk_Back
     tiling=NoTile
     imageSize=<600 600>
     margin=<ltrb 0 0 0 0>
     outlineSettings {
      bUseBrushTransparency=false
      roundingType=HalfHeightRadius
      width=0
      color {colorUseRule=UseColor_Specified, specifiedColor=<rgba 0 0 0 0>}
      cornerRadii=<0 0 0 0>
     }
     tintColor {colorUseRule=UseColor_Specified, specifiedColor=<rgba 1 1 1 1>}
     uVRegion {bIsValid=false, max=<0 0>, min=<0 0>}
    }
    renderTransform {angle=0, scale=<1 1>, shear=<0 0>, translation=<0 38>}
   }
   slot {class=ButtonSlot, properties {horizontalAlignment=HAlign_Fill, verticalAlignment=VAlign_Fill, padding=<ltrb 0 0 0 0>}}
  }
  BTN_PW_Down {
   widgetClassPath=&EditorUtilityButton
   parent=VerticalBox
   widget=BTN_PW_Down
   properties {
    toolTipText=S
    backgroundColor=<rgba 1 1 1 0.2>
    colorAndOpacity=<rgba 0.0865 0.1845 0.3515 0.7>
    widgetStyle {
     clickedSlateSound {resourceObject=null}
     disabled {
      bIsDynamicallyLoaded=false
      drawAs=RoundedBox
      imageType=FullColor
      mirroring=NoMirror
      resourceName=None
      resourceObject=null
      tiling=NoTile
      imageSize=<160 160>
      margin=<ltrb 0 0 0 0>
      outlineSettings {
       bUseBrushTransparency=false
       roundingType=FixedRadius
       width=3
       color {colorUseRule=UseColor_Specified, specifiedColor=<rgba 0 0 0 0.2>}
       cornerRadii=<20 20 20 20>
      }
      tintColor {colorUseRule=UseColor_Specified, specifiedColor=<rgba 0.0395 0.0395 0.0395 1>}
      uVRegion {bIsValid=false, max=<0 0>, min=<0 0>}
     }
     disabledForeground {colorUseRule=UseColor_ColorTable, specifiedColor=<rgba 0 0 0 0>}
     hovered {
      bIsDynamicallyLoaded=false
      drawAs=RoundedBox
      imageType=FullColor
      mirroring=NoMirror
      resourceName=None
      resourceObject=null
      tiling=NoTile
      imageSize=<160 160>
      margin=<ltrb 0 0 0 0>
      outlineSettings {
       bUseBrushTransparency=false
       roundingType=FixedRadius
       width=3
       color {colorUseRule=UseColor_Specified, specifiedColor=<rgba 0.7222 0.7222 0.7222 0.2>}
       cornerRadii=<20 20 20 20>
      }
      tintColor {colorUseRule=UseColor_Specified, specifiedColor=<rgba 0.0953 0.0953 0.0953 1>}
      uVRegion {bIsValid=false, max=<0 0>, min=<0 0>}
     }
     hoveredForeground {colorUseRule=UseColor_ColorTable, specifiedColor=<rgba 0 0 0 0>}
     hoveredSlateSound {resourceObject=null}
     normal {
      bIsDynamicallyLoaded=false
      drawAs=RoundedBox
      imageType=FullColor
      mirroring=NoMirror
      resourceName=None
      resourceObject=null
      tiling=NoTile
      imageSize=<160 160>
      margin=<ltrb 0 0 0 0>
      outlineSettings {
       bUseBrushTransparency=false
       roundingType=FixedRadius
       width=15
       color {colorUseRule=UseColor_Specified, specifiedColor=<rgba 0.7222 0.7222 0.7222 0.1>}
       cornerRadii=<30 30 30 30>
      }
      tintColor {colorUseRule=UseColor_Specified, specifiedColor=<rgba 0.0395 0.0395 0.0395 1>}
      uVRegion {bIsValid=false, max=<0 0>, min=<0 0>}
     }
     normalForeground {colorUseRule=UseColor_ColorTable, specifiedColor=<rgba 0 0 0 0>}
     normalPadding=<ltrb 12 1.5 12 1.5>
     pressed {
      bIsDynamicallyLoaded=false
      drawAs=RoundedBox
      imageType=FullColor
      mirroring=NoMirror
      resourceName=None
      resourceObject=null
      tiling=NoTile
      imageSize=<160 160>
      margin=<ltrb 0 0 0 0>
      outlineSettings {
       bUseBrushTransparency=false
       roundingType=FixedRadius
       width=3
       color {colorUseRule=UseColor_Specified, specifiedColor=<rgba 0.7222 0.7222 0.7222 0.2>}
       cornerRadii=<20 20 20 20>
      }
      tintColor {colorUseRule=UseColor_Specified, specifiedColor=<rgba 0.0284 0.0284 0.0284 1>}
      uVRegion {bIsValid=false, max=<0 0>, min=<0 0>}
     }
     pressedForeground {colorUseRule=UseColor_ColorTable, specifiedColor=<rgba 0 0 0 0>}
     pressedPadding=<ltrb 12 2.5 12 0.5>
     pressedSlateSound {resourceObject=null}
    }
   }
   slot {class=VerticalBoxSlot, properties {padding=<ltrb 20 20 20 20>}}
  }
  Image {
   widgetClassPath=&Image
   bIsVariable=true
   parent=BTN_PW_Down
   widget=Image
   properties {brush {
     bIsDynamicallyLoaded=false
     drawAs=Image
     imageType=FullColor
     mirroring=NoMirror
     resourceName=None
     resourceObject=/AnimInEdTools/Pickwalk_ControlRig/icons/AnimToolbox_UI_Pickwalk_Down
     tiling=NoTile
     imageSize=<600 600>
     margin=<ltrb 0 0 0 0>
     outlineSettings {
      bUseBrushTransparency=false
      roundingType=HalfHeightRadius
      width=0
      color {colorUseRule=UseColor_Specified, specifiedColor=<rgba 0 0 0 0>}
      cornerRadii=<0 0 0 0>
     }
     tintColor {colorUseRule=UseColor_Specified, specifiedColor=<rgba 1 1 1 1>}
     uVRegion {bIsValid=false, max=<0 0>, min=<0 0>}
    }}
   slot {class=ButtonSlot, properties {horizontalAlignment=HAlign_Fill, verticalAlignment=VAlign_Fill}}
  }
  VerticalBox_1 {widgetClassPath=&VerticalBox, parent=HorizontalBox_126, widget=VerticalBox_1, slot {class=HorizontalBoxSlot}}
  BTN_SelPart {
   widgetClassPath=&EditorUtilityButton
   parent=VerticalBox_1
   widget=BTN_SelPart
   properties {
    toolTipText=E
    backgroundColor=<rgba 1 1 1 0.2>
    colorAndOpacity=<rgba 0.063 0.2502 0.1274 0.7>
    widgetStyle {
     clickedSlateSound {resourceObject=null}
     disabled {
      bIsDynamicallyLoaded=false
      drawAs=RoundedBox
      imageType=FullColor
      mirroring=NoMirror
      resourceName=None
      resourceObject=null
      tiling=NoTile
      imageSize=<64 64>
      margin=<ltrb 0 0 0 0>
      outlineSettings {
       bUseBrushTransparency=false
       roundingType=FixedRadius
       width=3
       color {colorUseRule=UseColor_Specified, specifiedColor=<rgba 0 0 0 0.2>}
       cornerRadii=<20 20 20 20>
      }
      tintColor {colorUseRule=UseColor_Specified, specifiedColor=<rgba 0.0395 0.0395 0.0395 1>}
      uVRegion {bIsValid=false, max=<0 0>, min=<0 0>}
     }
     disabledForeground {colorUseRule=UseColor_ColorTable, specifiedColor=<rgba 0 0 0 0>}
     hovered {
      bIsDynamicallyLoaded=false
      drawAs=RoundedBox
      imageType=FullColor
      mirroring=NoMirror
      resourceName=None
      resourceObject=null
      tiling=NoTile
      imageSize=<64 64>
      margin=<ltrb 0 0 0 0>
      outlineSettings {
       bUseBrushTransparency=false
       roundingType=FixedRadius
       width=3
       color {colorUseRule=UseColor_Specified, specifiedColor=<rgba 0.7222 0.7222 0.7222 0.2>}
       cornerRadii=<20 20 20 20>
      }
      tintColor {colorUseRule=UseColor_Specified, specifiedColor=<rgba 0.0953 0.0953 0.0953 1>}
      uVRegion {bIsValid=false, max=<0 0>, min=<0 0>}
     }
     hoveredForeground {colorUseRule=UseColor_ColorTable, specifiedColor=<rgba 0 0 0 0>}
     hoveredSlateSound {resourceObject=null}
     normal {
      bIsDynamicallyLoaded=false
      drawAs=RoundedBox
      imageType=FullColor
      mirroring=NoMirror
      resourceName=None
      resourceObject=null
      tiling=NoTile
      imageSize=<64 64>
      margin=<ltrb 0 0 0 0>
      outlineSettings {
       bUseBrushTransparency=false
       roundingType=FixedRadius
       width=15
       color {colorUseRule=UseColor_Specified, specifiedColor=<rgba 0.7222 0.7222 0.7222 0.1>}
       cornerRadii=<30 30 30 30>
      }
      tintColor {colorUseRule=UseColor_Specified, specifiedColor=<rgba 0.0395 0.0395 0.0395 1>}
      uVRegion {bIsValid=false, max=<0 0>, min=<0 0>}
     }
     normalForeground {colorUseRule=UseColor_ColorTable, specifiedColor=<rgba 0 0 0 0>}
     normalPadding=<ltrb 12 1.5 12 1.5>
     pressed {
      bIsDynamicallyLoaded=false
      drawAs=RoundedBox
      imageType=FullColor
      mirroring=NoMirror
      resourceName=None
      resourceObject=null
      tiling=NoTile
      imageSize=<64 64>
      margin=<ltrb 0 0 0 0>
      outlineSettings {
       bUseBrushTransparency=false
       roundingType=FixedRadius
       width=3
       color {colorUseRule=UseColor_Specified, specifiedColor=<rgba 0.7222 0.7222 0.7222 0.2>}
       cornerRadii=<20 20 20 20>
      }
      tintColor {colorUseRule=UseColor_Specified, specifiedColor=<rgba 0.0284 0.0284 0.0284 1>}
      uVRegion {bIsValid=false, max=<0 0>, min=<0 0>}
     }
     pressedForeground {colorUseRule=UseColor_ColorTable, specifiedColor=<rgba 0 0 0 0>}
     pressedPadding=<ltrb 12 2.5 12 0.5>
     pressedSlateSound {resourceObject=null}
    }
   }
   slot {class=VerticalBoxSlot, properties {padding=<ltrb 20 20 20 20>}}
  }
  Image_1 {
   widgetClassPath=&Image
   bIsVariable=true
   parent=BTN_SelPart
   widget=Image_1
   properties {brush {
     bIsDynamicallyLoaded=false
     drawAs=Image
     imageType=FullColor
     mirroring=NoMirror
     resourceName=None
     resourceObject=/AnimInEdTools/Pickwalk_ControlRig/icons/AnimToolbox_UI_Pickwalk_SelectPart
     tiling=NoTile
     imageSize=<600 600>
     margin=<ltrb 0 0 0 0>
     outlineSettings {
      bUseBrushTransparency=false
      roundingType=HalfHeightRadius
      width=0
      color {colorUseRule=UseColor_Specified, specifiedColor=<rgba 0 0 0 0>}
      cornerRadii=<0 0 0 0>
     }
     tintColor {colorUseRule=UseColor_Specified, specifiedColor=<rgba 1 1 1 1>}
     uVRegion {bIsValid=false, max=<0 0>, min=<0 0>}
    }}
   slot {class=ButtonSlot, properties {horizontalAlignment=HAlign_Fill, verticalAlignment=VAlign_Fill}}
  }
  BTN_PW_Right {
   widgetClassPath=&EditorUtilityButton
   parent=VerticalBox_1
   widget=BTN_PW_Right
   properties {
    toolTipText=D
    backgroundColor=<rgba 1 1 1 0.2>
    colorAndOpacity=<rgba 0.0865 0.1845 0.3515 0.7>
    widgetStyle {
     clickedSlateSound {resourceObject=null}
     disabled {
      bIsDynamicallyLoaded=false
      drawAs=RoundedBox
      imageType=FullColor
      mirroring=NoMirror
      resourceName=None
      resourceObject=null
      tiling=NoTile
      imageSize=<64 64>
      margin=<ltrb 0 0 0 0>
      outlineSettings {
       bUseBrushTransparency=false
       roundingType=FixedRadius
       width=3
       color {colorUseRule=UseColor_Specified, specifiedColor=<rgba 0 0 0 0.2>}
       cornerRadii=<20 20 20 20>
      }
      tintColor {colorUseRule=UseColor_Specified, specifiedColor=<rgba 0.0395 0.0395 0.0395 1>}
      uVRegion {bIsValid=false, max=<0 0>, min=<0 0>}
     }
     disabledForeground {colorUseRule=UseColor_ColorTable, specifiedColor=<rgba 0 0 0 0>}
     hovered {
      bIsDynamicallyLoaded=false
      drawAs=RoundedBox
      imageType=FullColor
      mirroring=NoMirror
      resourceName=None
      resourceObject=null
      tiling=NoTile
      imageSize=<64 64>
      margin=<ltrb 0 0 0 0>
      outlineSettings {
       bUseBrushTransparency=false
       roundingType=FixedRadius
       width=3
       color {colorUseRule=UseColor_Specified, specifiedColor=<rgba 0.7222 0.7222 0.7222 0.2>}
       cornerRadii=<20 20 20 20>
      }
      tintColor {colorUseRule=UseColor_Specified, specifiedColor=<rgba 0.0953 0.0953 0.0953 1>}
      uVRegion {bIsValid=false, max=<0 0>, min=<0 0>}
     }
     hoveredForeground {colorUseRule=UseColor_ColorTable, specifiedColor=<rgba 0 0 0 0>}
     hoveredSlateSound {resourceObject=null}
     normal {
      bIsDynamicallyLoaded=false
      drawAs=RoundedBox
      imageType=FullColor
      mirroring=NoMirror
      resourceName=None
      resourceObject=null
      tiling=NoTile
      imageSize=<64 64>
      margin=<ltrb 0 0 0 0>
      outlineSettings {
       bUseBrushTransparency=false
       roundingType=FixedRadius
       width=15
       color {colorUseRule=UseColor_Specified, specifiedColor=<rgba 0.7222 0.7222 0.7222 0.1>}
       cornerRadii=<30 30 30 30>
      }
      tintColor {colorUseRule=UseColor_Specified, specifiedColor=<rgba 0.0395 0.0395 0.0395 1>}
      uVRegion {bIsValid=false, max=<0 0>, min=<0 0>}
     }
     normalForeground {colorUseRule=UseColor_ColorTable, specifiedColor=<rgba 0 0 0 0>}
     normalPadding=<ltrb 12 1.5 12 1.5>
     pressed {
      bIsDynamicallyLoaded=false
      drawAs=RoundedBox
      imageType=FullColor
      mirroring=NoMirror
      resourceName=None
      resourceObject=null
      tiling=NoTile
      imageSize=<64 64>
      margin=<ltrb 0 0 0 0>
      outlineSettings {
       bUseBrushTransparency=false
       roundingType=FixedRadius
       width=3
       color {colorUseRule=UseColor_Specified, specifiedColor=<rgba 0.7222 0.7222 0.7222 0.2>}
       cornerRadii=<20 20 20 20>
      }
      tintColor {colorUseRule=UseColor_Specified, specifiedColor=<rgba 0.0284 0.0284 0.0284 1>}
      uVRegion {bIsValid=false, max=<0 0>, min=<0 0>}
     }
     pressedForeground {colorUseRule=UseColor_ColorTable, specifiedColor=<rgba 0 0 0 0>}
     pressedPadding=<ltrb 12 2.5 12 0.5>
     pressedSlateSound {resourceObject=null}
    }
   }
   slot {class=VerticalBoxSlot, properties {padding=<ltrb 20 20 20 20>}}
  }
  Image_2 {
   widgetClassPath=&Image
   bIsVariable=true
   parent=BTN_PW_Right
   widget=Image_2
   properties {brush {
     bIsDynamicallyLoaded=false
     drawAs=Image
     imageType=FullColor
     mirroring=NoMirror
     resourceName=None
     resourceObject=/AnimInEdTools/Pickwalk_ControlRig/icons/AnimToolbox_UI_Pickwalk_Right
     tiling=NoTile
     imageSize=<600 600>
     margin=<ltrb 0 0 0 0>
     outlineSettings {
      bUseBrushTransparency=false
      roundingType=HalfHeightRadius
      width=0
      color {colorUseRule=UseColor_Specified, specifiedColor=<rgba 0 0 0 0>}
      cornerRadii=<0 0 0 0>
     }
     tintColor {colorUseRule=UseColor_Specified, specifiedColor=<rgba 1 1 1 1>}
     uVRegion {bIsValid=false, max=<0 0>, min=<0 0>}
    }}
   slot {class=ButtonSlot, properties {horizontalAlignment=HAlign_Fill, verticalAlignment=VAlign_Fill}}
  }
  BTN_SpecialShit {
   widgetClassPath=&EditorUtilityButton
   parent=VerticalBox_1
   widget=BTN_SpecialShit
   properties {
    toolTipText=C
    backgroundColor=<rgba 1 1 1 0.2>
    colorAndOpacity=<rgba 0.1946 0.1301 0.3515 0.7>
    widgetStyle {
     clickedSlateSound {resourceObject=null}
     disabled {
      bIsDynamicallyLoaded=false
      drawAs=RoundedBox
      imageType=FullColor
      mirroring=NoMirror
      resourceName=None
      resourceObject=null
      tiling=NoTile
      imageSize=<64 64>
      margin=<ltrb 0 0 0 0>
      outlineSettings {
       bUseBrushTransparency=false
       roundingType=FixedRadius
       width=3
       color {colorUseRule=UseColor_Specified, specifiedColor=<rgba 0 0 0 0.2>}
       cornerRadii=<20 20 20 20>
      }
      tintColor {colorUseRule=UseColor_Specified, specifiedColor=<rgba 0.0395 0.0395 0.0395 1>}
      uVRegion {bIsValid=false, max=<0 0>, min=<0 0>}
     }
     disabledForeground {colorUseRule=UseColor_ColorTable, specifiedColor=<rgba 0 0 0 0>}
     hovered {
      bIsDynamicallyLoaded=false
      drawAs=RoundedBox
      imageType=FullColor
      mirroring=NoMirror
      resourceName=None
      resourceObject=null
      tiling=NoTile
      imageSize=<64 64>
      margin=<ltrb 0 0 0 0>
      outlineSettings {
       bUseBrushTransparency=false
       roundingType=FixedRadius
       width=3
       color {colorUseRule=UseColor_Specified, specifiedColor=<rgba 0.7222 0.7222 0.7222 0.2>}
       cornerRadii=<20 20 20 20>
      }
      tintColor {colorUseRule=UseColor_Specified, specifiedColor=<rgba 0.0953 0.0953 0.0953 1>}
      uVRegion {bIsValid=false, max=<0 0>, min=<0 0>}
     }
     hoveredForeground {colorUseRule=UseColor_ColorTable, specifiedColor=<rgba 0 0 0 0>}
     hoveredSlateSound {resourceObject=null}
     normal {
      bIsDynamicallyLoaded=false
      drawAs=RoundedBox
      imageType=FullColor
      mirroring=NoMirror
      resourceName=None
      resourceObject=null
      tiling=NoTile
      imageSize=<64 64>
      margin=<ltrb 0 0 0 0>
      outlineSettings {
       bUseBrushTransparency=false
       roundingType=FixedRadius
       width=15
       color {colorUseRule=UseColor_Specified, specifiedColor=<rgba 0.7222 0.7222 0.7222 0.1>}
       cornerRadii=<30 30 30 30>
      }
      tintColor {colorUseRule=UseColor_Specified, specifiedColor=<rgba 0.0395 0.0395 0.0395 1>}
      uVRegion {bIsValid=false, max=<0 0>, min=<0 0>}
     }
     normalForeground {colorUseRule=UseColor_ColorTable, specifiedColor=<rgba 0 0 0 0>}
     normalPadding=<ltrb 12 1.5 12 1.5>
     pressed {
      bIsDynamicallyLoaded=false
      drawAs=RoundedBox
      imageType=FullColor
      mirroring=NoMirror
      resourceName=None
      resourceObject=null
      tiling=NoTile
      imageSize=<64 64>
      margin=<ltrb 0 0 0 0>
      outlineSettings {
       bUseBrushTransparency=false
       roundingType=FixedRadius
       width=3
       color {colorUseRule=UseColor_Specified, specifiedColor=<rgba 0.7222 0.7222 0.7222 0.2>}
       cornerRadii=<20 20 20 20>
      }
      tintColor {colorUseRule=UseColor_Specified, specifiedColor=<rgba 0.0284 0.0284 0.0284 1>}
      uVRegion {bIsValid=false, max=<0 0>, min=<0 0>}
     }
     pressedForeground {colorUseRule=UseColor_ColorTable, specifiedColor=<rgba 0 0 0 0>}
     pressedPadding=<ltrb 12 2.5 12 0.5>
     pressedSlateSound {resourceObject=null}
    }
   }
   slot {class=VerticalBoxSlot, properties {padding=<ltrb 20 20 20 20>}}
  }
  Image_3 {
   widgetClassPath=&Image
   bIsVariable=true
   parent=BTN_SpecialShit
   widget=Image_3
   properties {brush {
     bIsDynamicallyLoaded=false
     drawAs=Image
     imageType=FullColor
     mirroring=NoMirror
     resourceName=None
     resourceObject=/AnimInEdTools/Pickwalk_ControlRig/icons/AnimToolbox_UI_Pickwalk_Special
     tiling=NoTile
     imageSize=<600 600>
     margin=<ltrb 0 0 0 0>
     outlineSettings {
      bUseBrushTransparency=false
      roundingType=HalfHeightRadius
      width=0
      color {colorUseRule=UseColor_Specified, specifiedColor=<rgba 0 0 0 0>}
      cornerRadii=<0 0 0 0>
     }
     tintColor {colorUseRule=UseColor_Specified, specifiedColor=<rgba 1 1 1 1>}
     uVRegion {bIsValid=false, max=<0 0>, min=<0 0>}
    }}
   slot {class=ButtonSlot, properties {horizontalAlignment=HAlign_Fill, verticalAlignment=VAlign_Fill}}
  }
 ]
}
